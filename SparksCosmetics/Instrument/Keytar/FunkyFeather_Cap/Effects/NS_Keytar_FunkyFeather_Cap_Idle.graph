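# Niagara idle effect for the FunkyFeather Cap keytar cosmetic, shedding fluttering mesh and sprite particles from a Falling_Money emitter.
# Spawn rate follows a system speed value and surges with User.Landing, while context and variant switches choose spawn boxes, velocities, colors and visibility.
# Gravity, curl noise, aerodynamic drag and repulsion shape the fall, with distance-based shrinking and velocity-dependent world/local space blending.
# A debug emitter is disabled.
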NiagaraSystem /SparksCosmetics/Instrument/Keytar/FunkyFeather_Cap/Effects/NS_Keytar_FunkyFeather_Cap_Idle
dependencies {
 usedDataInterfaces [
  &NiagaraDataInterfaceArrayColor
  &NiagaraDataInterfaceCurve
  &NiagaraDataInterfaceVectorCurve
  &NiagaraDataInterfaceColorCurve
 ]
 usedDynamicInputs [
  &MakeFloatFromVector
  &MakeIntFromFloat
  &Boolean_AndOperation
  &Max_Float
  &RemapRange
  &Abs_Float
 ]
 usedModules [
  &self:SystemSpawnScript.SystemScriptSource.SystemScriptGraph.NiagaraNodeAssignment_0.SetVariables_#1
  &self:SystemSpawnScript.SystemScriptSource.SystemScriptGraph.NiagaraNodeAssignment_2.SetVariables_#2
  &SystemState
  &self:SystemSpawnScript.SystemScriptSource.SystemScriptGraph.NiagaraNodeAssignment_3.SetVariables_#3
  &self:SystemSpawnScript.SystemScriptSource.SystemScriptGraph.NiagaraNodeAssignment_4.SetVariables_#4
  &self:SystemSpawnScript.SystemScriptSource.SystemScriptGraph.NiagaraNodeAssignment_1.SetVariables_#5
  &self:SystemSpawnScript.SystemScriptSource.SystemScriptGraph.NiagaraNodeAssignment_5.SetVariables_#6
  &self:SystemSpawnScript.SystemScriptSource.SystemScriptGraph.NiagaraNodeAssignment_6.SetVariables_#7
  &self:SystemSpawnScript.SystemScriptSource.SystemScriptGraph.NiagaraNodeAssignment_7.SetVariables_#8
  &BlendSpaceHelper
 ]
}
emitters {
 DEBUG_VALUES {
  bEnabled=false
  emitterUpdateScript [
   &EmitterState()
   &SpawnBurst_Instantaneous(Use Spawn Probability=false, UseLoopCountLimit=true)
  ]
  particleSpawnScript [&InitializeParticle(
    Color Mode=::NewEnumerator0
    Lifetime=5
    Position Offset=<0 0 100>
    Position Offset Coordinate Space=::World
    UsePositionOffset=true
    Write Material Random=true
   )]
  particleUpdateScript [
   &ParticleState(
    Kill Particles When Lifetime Has Elapsed=false
    Let Infinitely Lived Particles Die When Emitter Deactivates=true
    Loop Particles Lifetime=true
   )
   &DynamicMaterialParameters(
    Index 0 Param 1=&VectorLength(Vector=$Engine.Owner.Velocity)
    Module.Index0Float0Write=true
    Module.Index0Float1Write=true
    Module.Index0Float2Write=true
    Module.Index0Float3Write=true
   )
   &Color(Color=&ColorFromCurve(
     ColorCurve=DataInterface() {propertyValues {
       AlphaCurve {keys [<at 0 1 RCIM_Cubic>]}
       BlueCurve {keys [<at 0 0 RCIM_Cubic>,<at 0.5 0>,<at 1 1 RCIM_Cubic>]}
       GreenCurve {keys [<at 0 0 RCIM_Cubic>,<at 0.5 1>,<at 1 0 RCIM_Cubic>]}
       RedCurve {keys [<at 0 1 RCIM_Cubic>,<at 0.5 0>,<at 1 0 RCIM_Cubic>]}
      }}
     CurveIndex=$System.ZVelocityScale
    )) {enabled=false}
   &KillParticles(Kill Particles=&SetBoolByFloatComparison(A=$User.Landing, B=1, Comparison Type=::NewEnumerator2)) {enabled=false}
  ]
  rendererClasses [&NiagaraSpriteRendererProperties]
  renderers [{rendererClass=&NiagaraSpriteRendererProperties}]
 }
 Falling_Money {
  emitterUpdateScript [
   &EmitterState()
   &SpawnRate(
    Spawn Probability=0.75
    SpawnRate=&Multiply_Float(
     A=3
     B=&MergeHelper_FloatSwitch(0=$System.SpeedSpawnRate, 1=$System.SpeedSpawnRate, 2=?, 3=?, Index=$System.ContextInt, Num=::NewEnumerator0)
    )
    Use Spawn Probability=true
   ) {module=SpawnRate001}
   &SpawnRate(SpawnGroup=1, SpawnRate=&Multiply_Float(A=60, B=$User.Landing), Use Spawn Probability=false)
  ]
  particleSpawnScript [
   &PartitionParticles(Number Of Partitions=3)
   &InitializeParticle(
    Color=&MergeHelper_ColorSwitch(
     0=&SelectLinearColorFromArray(
      Array Sampling Mode=::NewEnumerator0
      Direct Array Index=$Particles.PartitionParticles.Partition
      Direct Array Mode=::NewEnumerator1
      Color Selection Array=DataInterface() {propertyValues {ColorData [<rgba 0.36 0.9 0.5535 1>, <rgba 1 0.4453 0.619 1>]}}
      Random Seed=0
     )
     1=<rgba 1 0.5467 0.15 1>
     2=<rgba 1 0.5467 0.15 1>
     3=?
     Index=$System.Variant Int
     Num=::NewEnumerator1
    )
    Lifetime=&RandomRangeFloat(Minimum=1.35, Maximum=1.45)
    Mesh Scale Mode=::NewEnumerator3
    Mesh Uniform Scale=&Multiply_Float(A=0.85, B=1)
    Sprite Size Mode=::NewEnumerator1
    Uniform Sprite Size Max=5
    UsePositionOffset=false
    Write Material Random=true
   )
   &ShapeLocation(
    Box Midpoint=&MergeHelper_VectorSwitch(
     0=&MergeHelper_VectorSwitch(0=<-22 0.5 0>, 1=<-15 0.5 -10>, 2=<-0.5 0.5 4>, 3=?, Index=$Particles.PartitionParticles.Partition, Num=::NewEnumerator1)
     1=&MergeHelper_VectorSwitch(0=<-22 0.5 0>, 1=<-0.5 0.5 4>, 2=<-0.5 0.5 4>, 3=?, Index=$Particles.PartitionParticles.Partition, Num=::NewEnumerator1)
     2=?
     3=?
     Index=$System.ContextInt
     Num=::NewEnumerator0
    )
    Box Size=&MergeHelper_VectorSwitch(
     0=&MergeHelper_VectorSwitch(0=<1 2 15>, 1=<1 1 5>, 2=<10 1 1>, 3=?, Index=$Particles.PartitionParticles.Partition, Num=::NewEnumerator1)
     1=&MergeHelper_VectorSwitch(0=<1 2 15>, 1=<10 1 1>, 2=<10 1 1>, 3=?, Index=$Particles.PartitionParticles.Partition, Num=::NewEnumerator1)
     2=?
     3=?
     Index=$System.ContextInt
     Num=::NewEnumerator0
    )
    MaskSpawnGroup=false
    Offset Mode=::NewEnumerator0
    Override Seed=false
    Shape Primitive=::NewEnumerator2
    Surface Only=false
    UseOwnerScale=true
   )
   &InitialMeshOrientation(Use Rotation Vector=false)
   &AddVelocity(
    Rotation Coordinate Space=::World
    Velocity=&RandomRangeVector(Minimum=<0 0 110>, Maximum=<0 0 125>)
    Velocity Speed Scale=&Add_Float(
     A=&MergeHelper_FloatSwitch(0=0, 1=0.5, 2=?, 3=?, Index=$System.ContextInt, Num=::NewEnumerator0)
     B=&Multiply_Float(A=2, B=$User.Landing)
    )
   ) {module=AddVelocity002}
   &AddVelocity(
    Module.Use Falloff Distance=false
    Origin Offset=&MergeHelper_VectorSwitch(0=<0 0 40>, 1=<0 -50 55>, 2=?, 3=?, Index=$System.ContextInt, Num=::NewEnumerator0)
    Override Seed=false
    Velocity Mode=::NewEnumerator1
    Velocity Speed=&Multiply_Float(
     A=&RandomRangeFloat(Minimum=35, Maximum=40)
     B=&MergeHelper_FloatSwitch(0=1, 1=1, 2=?, 3=?, Index=$System.ContextInt, Num=::NewEnumerator0)
    )
   )
   &InheritVelocity(
    Inherited Velocity Amount Scale=&MakeVector(X=1, Y=1, Z=$System.ZVelocityScale)
    Speed Limit=true
    UseSpeedThreshold=false
    Velocity Destination Space:ENiagaraCoordinateSpace=::World
   )
   /SetVariables(Particles.DistanceScale=1) {isSetParameters=true}
   /SetVariables(Particles.MeshIndex=&RandomRangeInt(Minimum=0, Maximum=2)) {isSetParameters=true}
   &CameraOffset(Camera Offset Amount=10) {enabled=false}
   /SetVariables(Particles.VisibilityTag=&MergeHelper_IntSwitch(0=0, 1=1, 2=1, 3=?, 4=?, 5=?, Index=$System.Variant Int, Num=::NewEnumerator1)) {isSetParameters=true}
  ]
  particleUpdateScript [
   &ParticleState()
   &UpdateMeshOrientation(
    Rotation Rate=&FloatFromCurve(FloatCurve=DataInterface() {propertyValues {Curve {
        defaultValue=1
        keys [{interpMode=RCIM_Cubic, tangentMode=RCTM_User, time=0, value=0}, {interpMode=RCIM_Cubic, tangentMode=RCTM_User, time=1, value=1}]
       }}})
    Rotation Vector=&RandomRangeVector(Minimum=<1 0 0>, Maximum=<1 0.2 0.2>)
   )
   /SetVariables(Particles.Drag:NiagaraFloat=&FloatFromCurve(
     FloatCurve=DataInterface() {propertyValues {Curve {keys [
         {arriveTangent=8.8724, interpMode=RCIM_Cubic, leaveTangent=8.8724, tangentMode=RCTM_User, time=0, value=0.2}
         {
          arriveTangent=-1.8207
          arriveTangentWeight=0.0483
          interpMode=RCIM_Cubic
          leaveTangent=-1.8207
          leaveTangentWeight=0.1448
          tangentMode=RCTM_User
          time=0.15
          value=1
         }
         <at 1 0.22 RCIM_Cubic>
        ]}}}
     Scale Curve=&RandomRangeFloat(Minimum=3.75, Maximum=4.25)
    )) {isSetParameters=true}
   /SetVariables(Particles.DistanceScale:NiagaraFloat=/Niagara/DynamicInputs/Math/Min_Float(
     A:NiagaraFloat=$Particles.DistanceScale
     B:NiagaraFloat=&FloatFromCurve(
      FloatCurve=DataInterface() {propertyValues {Curve {keys [
          {interpMode=RCIM_Cubic, leaveTangentWeight=0.4667, tangentMode=RCTM_User, time=0.8, value=1}
          {arriveTangent=-2.8825, interpMode=RCIM_Cubic, leaveTangent=-2.8825, tangentMode=RCTM_User, time=2, value=0}
         ]}}}
      CurveIndex=&Divide_Float(
       A=&VectorLength(Vector=&SubtractVectorFromPosition(Position=$Particles.Position, Vector=Linked() {linkedVariable {name=Engine.Emitter.SimulationPosition, type=NiagaraPosition}}))
       B=55
      )
     )
    )) {isSetParameters=true}
   &GravityForce(Gravity=&RandomRangeVector(Minimum=<0 0 -800>, Maximum=<1 1 -980>))
   &ScaleMeshSize(Scale Factor=&Multiply_VectorByFloat(
     Vector=&VectorFromCurve(
      VectorCurve=DataInterface() {propertyValues {XCurve {keys [<at 0 1>,<at 1 1>]}, YCurve {keys [<at 0 1>,<at 1 1>]}, ZCurve {keys [<at 0 0.1>,<at 0.4 1>]}}}
      CurveIndex=$Particles.Age
      Scale Curve=&VectorFromFloat(Value=$System.EngineOwnerScale)
     )
     Float=&Multiply_Float(
      A=&FloatFromCurve(
       FloatCurve=DataInterface() {propertyValues {Curve {keys [
           {arriveTangent=4.1785, interpMode=RCIM_Cubic, leaveTangent=8.7018, tangentMode=RCTM_Break, time=0, value=0}
           {arriveTangent=0.0224, interpMode=RCIM_Cubic, leaveTangent=0.0224, tangentMode=RCTM_Break, time=0.3, value=1}
          ]}}}
       CurveIndex=$Particles.Age
       Scale Curve=&FloatFromCurve(FloatCurve=DataInterface() {propertyValues {Curve {keys [
            {interpMode=RCIM_Cubic, tangentMode=RCTM_Break, time=0.7, value=1}
            {arriveTangent=-6.5758, arriveTangentWeight=0.855, interpMode=RCIM_Cubic, tangentMode=RCTM_Break, time=1, value=0}
           ]}}})
      )
      B=$Particles.DistanceScale
     )
    ))
   &CurlNoiseForce(
    Bias Noise=false
    Noise Strength=&FloatFromCurve(
     FloatCurve=DataInterface() {propertyValues {Curve {
        defaultValue=1
        keys [{interpMode=RCIM_Cubic, tangentMode=RCTM_User, time=0, value=0}, {interpMode=RCIM_Cubic, tangentMode=RCTM_User, time=0.5, value=1}]
       }}}
     Scale Curve=500
    )
    Pan Noise=false
   )
   &AerodynamicDrag(
    Aerodynamic Drag:NiagaraFloat=$Particles.Drag
    Drag Shape Curve=DataInterface() {propertyValues {Curve {keys [
        {interpMode=RCIM_Cubic, tangentMode=RCTM_User, time=0, value=0.1}
        {arriveTangent=2.0799, interpMode=RCIM_Cubic, leaveTangent=2.0799, tangentMode=RCTM_User, time=1, value=1}
       ]}}}
    Lift Shape Curve=DataInterface() {propertyValues {Curve {keys [
        {interpMode=RCIM_Cubic, tangentMode=RCTM_User, time=0, value=0}
        {interpMode=RCIM_Cubic, tangentMode=RCTM_User, time=0.5, value=1}
        {tangentMode=RCTM_User, time=1, value=0}
       ]}}}
    Use Linear Drag=true
    Use Rotational Drag=true
   )
   &PointAttractionForce(
    Attraction Radius=35
    Attraction Strength=-75
    Attractor Position Offset:Vector3f=<10 0 20>
    Attractor Position Offset Coordinate Space:ENiagaraCoordinateSpace=::Local
    Kill Within Radius=false
    Module.Use Falloff=true
    UseAttractorOffset=true
   )
   &SolveForcesAndVelocity(Clamp Velocity=false, Limit Acceleration=false)
   &AlignSpriteToMeshOrientation(SetAlignment=true, SetFacing=true)
   &BlendSpace("Position Blend (World <-> Local)":NiagaraFloat=&MakeCustomFloatFromBool(
     Bool=&SetBoolByFloatComparison(A=&VectorLength(Vector=$Engine.Owner.Velocity), B=300)
     True Float=&FloatFromCurve(FloatCurve=DataInterface() {propertyValues {Curve {keys [
          {arriveTangent=-0.3205, interpMode=RCIM_Cubic, leaveTangent=-0.3205, tangentMode=RCTM_User, time=0, value=0.7}
          {arriveTangentWeight=0.5394, interpMode=RCIM_Cubic, tangentMode=RCTM_User, time=0.85, value=0}
         ]}}})
     False Float=&FloatFromCurve(FloatCurve=DataInterface() {propertyValues {Curve {keys [
          {arriveTangent=0.0521, interpMode=RCIM_Cubic, leaveTangent=0.0521, tangentMode=RCTM_User, time=0.1, value=0.5}
          {arriveTangentWeight=0.3269, interpMode=RCIM_Cubic, tangentMode=RCTM_User, time=0.5, value=0}
         ]}}})
     Translation Contribution:NiagaraFloat=1
     Rotation Contribution:NiagaraFloat=1
    ))
   &KillParticles(Kill Particles=&SetBoolByFloatComparison(A=$Particles.DistanceScale, B=0.001, Comparison Type=::NewEnumerator7))
   &DynamicMaterialParameters(
    Index 0 Param 1=&MakeCustomFloatFromBool(Bool=$User.bIsFrontEndPreview, True Float=0.05, False Float=0)
    Index 0 Param 2=&FloatFromCurve(FloatCurve=DataInterface() {propertyValues {Curve {
        defaultValue=1
        keys [
         {interpMode=RCIM_Cubic, leaveTangentWeight=0.9127, tangentMode=RCTM_User, time=0.2, value=0}
         {arriveTangentWeight=0.2716, interpMode=RCIM_Cubic, leaveTangentWeight=0.4526, tangentMode=RCTM_User, time=0.5, value=0.2}
         {arriveTangent=-0.0405, interpMode=RCIM_Cubic, leaveTangent=-0.0405, tangentMode=RCTM_User, time=1, value=0}
        ]
       }}})
    Module.Index0Float0Write=true
    Module.Index0Float1Write=true
    Module.Index0Float2Write=true
    Module.Index0Float3Write=true
   )
   /SetVariables(Particles.CameraOffset:NiagaraFloat=&FloatFromCurve(FloatCurve=DataInterface() {propertyValues {Curve {keys [<at 0 1>,<at 0.2 0>]}}}, Scale Curve=-5)) {isSetParameters=true}
  ]
  rendererClasses [&NiagaraMeshRendererProperties, &NiagaraSpriteRendererProperties]
  renderers [
   {rendererClass=&NiagaraMeshRendererProperties, rendererIndex=0}
   {rendererClass=&NiagaraMeshRendererProperties, rendererIndex=1}
   {rendererClass=&NiagaraSpriteRendererProperties, rendererIndex=2}
  ]
 }
}
summary {
 systemName=NS_Keytar_FunkyFeather_Cap_Idle
 emitters [
  {emitterName=Falling_Money, rendererClasses [&NiagaraMeshRendererProperties, &NiagaraSpriteRendererProperties]}
  {bEnabled=false, emitterName=DEBUG_VALUES, rendererClasses [&NiagaraSpriteRendererProperties]}
 ]
 userVariables [
  User.bIsFrontEnd {type=NiagaraBool}
  User.bIsFrontEndPreview {type=NiagaraBool}
  User.IsBackpack {type=NiagaraFloat, defaultValue=0, description=""}
  User.Landing {type=NiagaraFloat, defaultValue=0, description=""}
  User.Variant {type=NiagaraFloat, defaultValue=0, description=""}
 ]
}
