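# Niagara swing trail for the FunkyFeather Cap keytar pickaxe cosmetic.
# Anim-trail ribbon emitters switch on System Variant Int, take time-of-day color and alpha, and feed looping dynamic material parameters.
# Flickering skeletal-mesh sparkles spawn for some variants, while the VBucks emitter spawns variant-colored particles at a socket, inheriting swing velocity.
# Several duplicate color trails and a DEBUG_VALUES emitter are disabled.
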NiagaraSystem /SparksCosmetics/Instrument/Keytar/FunkyFeather_Cap/Effects/NS_Pickaxe_Keytar_FunkyFeather_Cap_SwingTrail
dependencies {
 usedDataInterfaces [
  &NiagaraDataInterfaceCurve
  &NiagaraDataInterfaceColorCurve
  &NiagaraDataInterfaceArrayColor
  &NiagaraDataInterfaceVectorCurve
  &NiagaraDataInterfaceArrayFloat
  &NiagaraDataInterfaceSocketReader
 ]
 usedDynamicInputs [&MergeHelper_FloatSwitch, &MakeIntFromFloat]
 usedModules [
  &self:SystemSpawnScript.SystemScriptSource.SystemScriptGraph.NiagaraNodeAssignment_2.SetVariables_#1
  &self:SystemSpawnScript.SystemScriptSource.SystemScriptGraph.NiagaraNodeAssignment_3.SetVariables_#2
  &self:SystemSpawnScript.SystemScriptSource.SystemScriptGraph.NiagaraNodeAssignment_7.SetVariables_#3
  &SystemState
  &self:SystemSpawnScript.SystemScriptSource.SystemScriptGraph.NiagaraNodeAssignment_4.SetVariables_#4
  &self:SystemSpawnScript.SystemScriptSource.SystemScriptGraph.NiagaraNodeAssignment_5.SetVariables_#5
  &self:SystemSpawnScript.SystemScriptSource.SystemScriptGraph.NiagaraNodeAssignment_6.SetVariables_#6
  &self:SystemSpawnScript.SystemScriptSource.SystemScriptGraph.NiagaraNodeAssignment_9.SetVariables_#7
  &self:SystemSpawnScript.SystemScriptSource.SystemScriptGraph.NiagaraNodeAssignment_10.SetVariables_#8
  &self:SystemSpawnScript.SystemScriptSource.SystemScriptGraph.NiagaraNodeAssignment_8.SetVariables_#9
  &self:SystemSpawnScript.SystemScriptSource.SystemScriptGraph.NiagaraNodeAssignment_0.SetVariables_#10
  &self:SystemSpawnScript.SystemScriptSource.SystemScriptGraph.NiagaraNodeAssignment_1.SetVariables_#11
 ]
}
emitters {
 Accent {
  emitterUpdateScript [
   &EmitterState()
   &FN_SpawnAnimTrail_v2(
    Number of Anim Trails:NiagaraInt32=&MergeHelper_IntSwitch(0=0, 1=0, 2=1, 3=?, 4=?, 5=?, Index=$System.Variant Int, Num=::NewEnumerator1)
    Spawn Interval:NiagaraInt32=&MergeHelper_IntSwitch(0=0, 1=1, 2=1, 3=?, 4=?, 5=?, Index=$System.Variant Int, Num=::NewEnumerator1)
   )
  ]
  particleSpawnScript [
   /Game/Effects/Niagara/Modules/AnimTrail/FN_InitializeAnimTrail/FN_InitializeAnimTrailBoneIndecies()
   /Game/Effects/Niagara/Modules/AnimTrail/FN_InitializeAnimTrail/FN_InitializeAnimTrail(Color=&MultiplyLinearColor(Linear Color=$System.TOD Color, Scale Factor=<1 1 1 1>), Lifetime=0.13)
   &AddVelocity(Velocity=<0 500 0>) {enabled=false}
   /SetVariables(Particles.VisibilityTag=$System.Variant Int) {isSetParameters=true, module=SetVariables_#12}
  ]
  particleUpdateScript [
   &ScaleRibbonWidth(Ribbon Width Scale=&FloatFromCurve(FloatCurve=DataInterface() {propertyValues {Curve {keys [
         {arriveTangent=-1.2579, interpMode=RCIM_Cubic, time=0, value=1}
         {arriveTangent=-0.9664, interpMode=RCIM_Cubic, leaveTangent=-0.9664, tangentMode=RCTM_User, time=1, value=0}
        ]}}})) {enabled=false}
   &ScaleColor(
    Scale Alpha=&FloatFromCurve(
     FloatCurve=DataInterface() {propertyValues {Curve {keys [
         {interpMode=RCIM_Cubic, leaveTangent=-0.5729, tangentMode=RCTM_Break, time=0, value=1}
         {arriveTangent=-0.5809, interpMode=RCIM_Cubic, leaveTangent=-0.5809, tangentMode=RCTM_User, time=0.5, value=0.15}
         {arriveTangent=-0.2139, interpMode=RCIM_Cubic, tangentMode=RCTM_Break, time=1, value=0}
        ]}}}
     Scale Curve=&FloatFromCurve(
      FloatCurve=DataInterface() {propertyValues {Curve {keys [
          <at 0 0.9 RCIM_Cubic>
          {arriveTangent=0.2255, interpMode=RCIM_Cubic, leaveTangent=0.2255, time=0.2025, value=0.95}
          {arriveTangentWeight=0.0858, interpMode=RCIM_Cubic, leaveTangentWeight=0.0558, tangentMode=RCTM_User, time=0.4434, value=1}
          {arriveTangentWeight=0.0572, interpMode=RCIM_Cubic, leaveTangentWeight=0.1462, tangentMode=RCTM_User, time=0.6, value=1}
          <at 1 0.9 RCIM_Cubic>
         ]}}}
      CurveIndex=Linked() {linkedVariable {name=NPC.FortniteNPC.TimeOfDay_0-1, type=NiagaraFloat}}
      Scale Curve=$System.AlphaFade
     )
    )
    ScaleA=true
    ScaleRGB=true
   )
   &FN_UpdateAnimTrail_v2()
   &SolveForcesAndVelocity(Clamp Velocity=false, Limit Acceleration=false)
   &DynamicMaterialParameters(
    Index 0 Param 1=0.6
    Index 1 Param 1=3
    Index 1 Param 3=&FloatFromCurve(
     FloatCurve=DataInterface() {propertyValues {Curve {defaultValue=0, keys [{tangentMode=RCTM_User, time=0, value=1}, {tangentMode=RCTM_User, time=1, value=0}]}}}
     CurveIndex=&Frac_Float(Float=&Multiply_Float(A=$Engine.Time, B=2))
    )
    Index 2 Param 1=15
    Index 2 Param 2=3
    Index 2 Param 3=0.5
    Index 2 Param 4=0.6
    Module.Index0Float0Write=true
    Module.Index0Float1Write=true
    Module.Index0Float2Write=true
    Module.Index0Float3Write=true
    Module.Index1Float0Write=true
    Module.Index1Float1Write=true
    Module.Index1Float2Write=true
    Module.Index1Float3Write=true
    Module.Index2Float0Write=true
    Module.Index2Float1Write=true
    Module.Index2Float2Write=false
    Module.Index2Float3Write=false
    Write Parameter Index 1=::NewEnumerator1
    Write Parameter Index 2=::NewEnumerator1
   )
  ]
  rendererClasses [&NiagaraRibbonRendererProperties]
  renderers [{rendererClass=&NiagaraRibbonRendererProperties}]
 }
 Accent_0_1 {
  emitterUpdateScript [
   &EmitterState()
   &FN_SpawnAnimTrail_v2(
    Number of Anim Trails:NiagaraInt32=&MergeHelper_IntSwitch(0=1, 1=1, 2=0, 3=?, 4=?, 5=?, Index=$System.Variant Int, Num=::NewEnumerator1)
    Spawn Interval:NiagaraInt32=&MergeHelper_IntSwitch(0=1, 1=1, 2=0, 3=?, 4=?, 5=?, Index=$System.Variant Int, Num=::NewEnumerator1)
   )
  ]
  particleSpawnScript [
   &FN_InitializeAnimTrailBoneIndecies()
   &FN_InitializeAnimTrail(Color=&MultiplyLinearColor(Linear Color=$System.TOD Color, Scale Factor=<1 1 1 0.7>), Lifetime=0.13)
   &AddVelocity(Velocity=<0 500 0>) {enabled=false}
   /SetVariables(Particles.VisibilityTag=$System.Variant Int) {isSetParameters=true, module=SetVariables_#12}
  ]
  particleUpdateScript [
   &ScaleRibbonWidth(Ribbon Width Scale=&FloatFromCurve(FloatCurve=DataInterface() {propertyValues {Curve {keys [
         {arriveTangent=-1.2579, interpMode=RCIM_Cubic, time=0, value=1}
         {arriveTangent=-0.9664, interpMode=RCIM_Cubic, leaveTangent=-0.9664, tangentMode=RCTM_User, time=1, value=0}
        ]}}}))
   &ScaleColor(
    Scale Alpha=&FloatFromCurve(
     FloatCurve=DataInterface() {propertyValues {Curve {keys [
         {interpMode=RCIM_Cubic, leaveTangent=-0.5729, tangentMode=RCTM_Break, time=0, value=1}
         {arriveTangent=-0.5809, interpMode=RCIM_Cubic, leaveTangent=-0.5809, tangentMode=RCTM_User, time=0.5, value=0.15}
         {arriveTangent=-0.2139, interpMode=RCIM_Cubic, tangentMode=RCTM_Break, time=1, value=0}
        ]}}}
     Scale Curve=&FloatFromCurve(
      FloatCurve=DataInterface() {propertyValues {Curve {keys [
          <at 0 0.9 RCIM_Cubic>
          {arriveTangent=0.2255, interpMode=RCIM_Cubic, leaveTangent=0.2255, time=0.2025, value=0.95}
          {arriveTangentWeight=0.0858, interpMode=RCIM_Cubic, leaveTangentWeight=0.0558, tangentMode=RCTM_User, time=0.4434, value=1}
          {arriveTangentWeight=0.0572, interpMode=RCIM_Cubic, leaveTangentWeight=0.1462, tangentMode=RCTM_User, time=0.6, value=1}
          <at 1 0.9 RCIM_Cubic>
         ]}}}
      CurveIndex=Linked() {linkedVariable {name=NPC.FortniteNPC.TimeOfDay_0-1, type=NiagaraFloat}}
      Scale Curve=$System.AlphaFade
     )
    )
    ScaleA=true
    ScaleRGB=true
   )
   &FN_UpdateAnimTrail_v2()
   &SolveForcesAndVelocity(Clamp Velocity=false, Limit Acceleration=false)
   &DynamicMaterialParameters(
    Index 0 Param 1=0.6
    Index 1 Param 1=3
    Index 1 Param 3=&FloatFromCurve(
     FloatCurve=DataInterface() {propertyValues {Curve {defaultValue=0, keys [{tangentMode=RCTM_User, time=0, value=1}, {tangentMode=RCTM_User, time=1, value=0}]}}}
     CurveIndex=&Frac_Float(Float=&Multiply_Float(A=$Engine.Time, B=2))
    )
    Index 2 Param 1=15
    Index 2 Param 2=3
    Index 2 Param 3=0.5
    Index 2 Param 4=0.6
    Module.Index0Float0Write=true
    Module.Index0Float1Write=true
    Module.Index0Float2Write=true
    Module.Index0Float3Write=true
    Module.Index1Float0Write=true
    Module.Index1Float1Write=true
    Module.Index1Float2Write=true
    Module.Index1Float3Write=true
    Module.Index2Float0Write=true
    Module.Index2Float1Write=true
    Module.Index2Float2Write=false
    Module.Index2Float3Write=false
    Write Parameter Index 1=::NewEnumerator1
    Write Parameter Index 2=::NewEnumerator1
   )
  ]
  rendererClasses [&NiagaraRibbonRendererProperties]
  renderers [{rendererClass=&NiagaraRibbonRendererProperties}]
 }
 DEBUG_VALUES {
  bEnabled=false
  emitterUpdateScript [
   &EmitterState()
   &SpawnBurst_Instantaneous(Use Spawn Probability=false, UseLoopCountLimit=true)
  ]
  particleSpawnScript [&InitializeParticle(
    Color Mode=::NewEnumerator0
    Position Offset=<0 0 100>
    Position Offset Coordinate Space=::World
    UsePositionOffset=true
    Write Material Random=true
   )]
  particleUpdateScript [
   &ParticleState()
   &DynamicMaterialParameters(
    Index 0 Param 1=&MakeFloatFromInt(Int=$System.Variant Int)
    Module.Index0Float0Write=true
    Module.Index0Float1Write=true
    Module.Index0Float2Write=true
    Module.Index0Float3Write=true
   )
   &Color(Color=&ColorFromCurve(
     ColorCurve=DataInterface() {propertyValues {
       AlphaCurve {keys [<at 0 1 RCIM_Cubic>]}
       BlueCurve {keys [<at 0 0 RCIM_Cubic>,<at 0.5 0>,<at 1 1 RCIM_Cubic>]}
       GreenCurve {keys [<at 0 0 RCIM_Cubic>,<at 0.5 1>,<at 1 0 RCIM_Cubic>]}
       RedCurve {keys [<at 0 1 RCIM_Cubic>,<at 0.5 0>,<at 1 0 RCIM_Cubic>]}
      }}
     CurveIndex=$System.ZVelocityScale
    )) {enabled=false}
   &KillParticles(Kill Particles=&SetBoolByFloatComparison(A=$User.Landing, B=1, Comparison Type=::NewEnumerator2)) {enabled=false}
  ]
  rendererClasses [&NiagaraSpriteRendererProperties]
  renderers [{rendererClass=&NiagaraSpriteRendererProperties}]
 }
 Flares_Sparkles {
  emitterUpdateScript [
   &EmitterState(Life Cycle Mode=::NewEnumerator1, Loop Behavior=::NewEnumerator1, Loop Duration=0.5, UseLoopDelay=false)
   &SpawnRate(
    Spawn Probability=&SelectFloatFromArray(Direct Array Index=$System.Variant Int, Float Selection Array=DataInterface() {propertyValues {FloatData [0,1,1]}})
    SpawnGroup=1
    SpawnRate=&RandomRangeFloat(Minimum=30, Maximum=50, Evaluation Type=::NewEnumerator1)
    Use Spawn Probability=true
   ) {module=SpawnRate001}
  ]
  particleSpawnScript [
   /SetVariables(Particles.VisibilityTag=&SpawnGroup()) {isSetParameters=true}
   &InitializeParticle(
    Color=<rgba 1 0.5504 0.0472 0.2>
    Lifetime=&FloatFromCurve(
     FloatCurve=DataInterface() {propertyValues {Curve {keys [<at 0 0.25>,<at 0.8 0.37>,<at 1 0.48>]}}}
     CurveIndex=&RandomRangeFloat(Minimum=0, Maximum=1)
    )
    Sprite Size Mode=::NewEnumerator3
    Uniform Sprite Size=&RandomRangeFloat(Minimum=6, Maximum=10)
    UsePositionOffset=false
    Write Material Random=true
   )
   &CameraOffset(Camera Offset Amount=10)
   &SkeletalMeshLocation(MaskSpawnGroup=false, Mesh Sampling Type=::NewEnumerator5, Skeletal Mesh=$User.SkeletalMesh, Socket Sampling Mode=::NewEnumerator5)
   &ShapeLocation(
    Cylinder Height=35
    Cylinder Height Midpoint=0
    Cylinder Radius=3
    MaskSpawnGroup=false
    Override Seed=false
    Shape Primitive=::NewEnumerator1
    Surface Only=false
    UseOwnerScale=true
   )
   &InheritVelocity(
    Inherited Velocity Speed Limit:NiagaraFloat=&RandomRangeFloat(Minimum=50, Maximum=75)
    Speed Limit=true
    UseSpeedThreshold=false
   )
   &AddVelocity(Velocity=<0 300 0>) {enabled=false}
  ]
  particleUpdateScript [
   &ParticleState()
   /SetVariables(Particles.FlickerWaveform:NiagaraFloat=&Waveform(
     Global Phase=&Add_Float(A=$Engine.Time, B=&RandomRangeFloat(Minimum=0, Maximum=1))
     "Waveform [1]"=::NewEnumerator0
     "[1] Frequency"=6
     "[1] Amplitude Min/Max"=<0 1>
    )) {isSetParameters=true}
   &ScaleColor(
    Scale Alpha=&Multiply_Float(
     A=&FloatFromCurve(
      FloatCurve=DataInterface() {propertyValues {Curve {
         defaultValue=1
         keys [
          {arriveTangent=6.7928, interpMode=RCIM_Cubic, leaveTangent=6.7928, leaveTangentWeight=0.4012, tangentMode=RCTM_User, time=0, value=0}
          {arriveTangentWeight=0.0442, interpMode=RCIM_Cubic, leaveTangentWeight=0.3945, tangentMode=RCTM_User, time=0.1477, value=1}
          {arriveTangentWeight=0.1189, interpMode=RCIM_Cubic, leaveTangentWeight=0.1652, tangentMode=RCTM_User, time=0.3389, value=1}
          <at 1 0 RCIM_Cubic>
         ]
        }}}
      Scale Curve=&RandomRangeFloat(Minimum=0.5, Maximum=1)
     )
     B=&FloatFromCurve(
      FloatCurve=DataInterface() {propertyValues {Curve {keys [<at 0 0.8>,<at 0.4974 1>,<at 1 0.8>]}}}
      CurveIndex=Linked() {linkedVariable {name=NPC.FortniteNPC.TimeOfDay_0-1, type=NiagaraFloat}}
     )
    )
    Scale RGB=&Multiply_VectorByFloat(
     Vector=&VectorFromFloat(Value=1)
     Float=&Lerp_Float(A=0.7, B=2, Alpha=$Particles.FlickerWaveform)
    )
    ScaleA=true
    ScaleRGB=true
   )
   &ScaleSpriteSize(Scale Factor=&Vector2DFromFloat(Value=&FloatFromCurve(
      FloatCurve=DataInterface() {propertyValues {Curve {
         defaultValue=0
         keys [
          {arriveTangent=8.3501, interpMode=RCIM_Cubic, leaveTangent=8.3501, tangentMode=RCTM_User, time=0, value=0}
          {arriveTangentWeight=0.0333, interpMode=RCIM_Cubic, leaveTangentWeight=0.2, tangentMode=RCTM_User, time=0.2136, value=1}
          {arriveTangent=-0.9157, arriveTangentWeight=0.2839, interpMode=RCIM_Cubic, leaveTangent=-0.9157, tangentMode=RCTM_User, time=1, value=0}
         ]
        }}}
      Scale Curve=&Lerp_Float(A=0.3, B=1.5, Alpha=$Particles.FlickerWaveform)
     ))) {module=ScaleSpriteSize001}
   &CurlNoiseForce(Noise Frequency=100, Noise Strength=200, Pan Noise=true, Pan Noise Field=<0 0 0.5>, Random Seed=10)
   &Drag(Drag=&RandomRangeFloat(Minimum=5, Maximum=6), Use Linear Drag=true, Use Rotational Drag=false)
   &SolveForcesAndVelocity(Clamp Velocity=false, Limit Acceleration=false)
  ]
  rendererClasses [&NiagaraSpriteRendererProperties]
  renderers [{rendererClass=&NiagaraSpriteRendererProperties}]
 }
 TrailColor001 {
  emitterUpdateScript [
   &EmitterState()
   &FN_SpawnAnimTrail_v2(
    Number of Anim Trails:NiagaraInt32=&MergeHelper_IntSwitch(0=1, 1=0, 2=0, 3=?, 4=?, 5=?, Index=$System.Variant Int, Num=::NewEnumerator1)
    Spawn Interval:NiagaraInt32=&MergeHelper_IntSwitch(0=1, 1=0, 2=0, 3=?, 4=?, 5=?, Index=$System.Variant Int, Num=::NewEnumerator1)
   )
  ]
  particleSpawnScript [
   &FN_InitializeAnimTrailBoneIndecies(First Bone or Socket Index=1)
   &FN_InitializeAnimTrail(
    AnimTrailWidth=1
    Color=&MultiplyLinearColor(Linear Color=$System.TOD Color, Scale Factor=<1 1 1 0.6>)
    Lifetime=0.25
   )
   &AddVelocity(Velocity=<0 500 0>) {enabled=false}
   /SetVariables(Particles.VisibilityTag=$System.Variant Int) {isSetParameters=true, module=SetVariables_#12}
  ]
  particleUpdateScript [
   &ScaleRibbonWidth(Ribbon Width Scale=&FloatFromCurve(FloatCurve=DataInterface() {propertyValues {Curve {keys [
         {arriveTangent=-1.2579, interpMode=RCIM_Cubic, time=0, value=1}
         {arriveTangent=-0.9664, interpMode=RCIM_Cubic, leaveTangent=-0.9664, tangentMode=RCTM_User, time=1, value=0}
        ]}}})) {enabled=false}
   &ScaleColor(
    Scale Alpha=&FloatFromCurve(
     FloatCurve=DataInterface() {propertyValues {Curve {keys [
         {interpMode=RCIM_Cubic, leaveTangent=-0.5729, tangentMode=RCTM_Break, time=0, value=1}
         {arriveTangent=-0.5809, interpMode=RCIM_Cubic, leaveTangent=-0.5809, tangentMode=RCTM_User, time=0.5, value=0.15}
         {arriveTangent=-0.2139, interpMode=RCIM_Cubic, tangentMode=RCTM_Break, time=1, value=0}
        ]}}}
     Scale Curve=&FloatFromCurve(
      FloatCurve=DataInterface() {propertyValues {Curve {keys [
          <at 0 0.9 RCIM_Cubic>
          {arriveTangent=0.2255, interpMode=RCIM_Cubic, leaveTangent=0.2255, time=0.2025, value=0.95}
          {arriveTangentWeight=0.0858, interpMode=RCIM_Cubic, leaveTangentWeight=0.0558, tangentMode=RCTM_User, time=0.4434, value=1}
          {arriveTangentWeight=0.0572, interpMode=RCIM_Cubic, leaveTangentWeight=0.1462, tangentMode=RCTM_User, time=0.6, value=1}
          <at 1 0.9 RCIM_Cubic>
         ]}}}
      CurveIndex=Linked() {linkedVariable {name=NPC.FortniteNPC.TimeOfDay_0-1, type=NiagaraFloat}}
      Scale Curve=$System.AlphaFade
     )
    )
    ScaleA=true
    ScaleRGB=true
   )
   &FN_UpdateAnimTrail_v2()
   &SolveForcesAndVelocity(Clamp Velocity=false, Limit Acceleration=false)
   &DynamicMaterialParameters(
    Index 0 Param 1=0.6
    Index 1 Param 1=3
    Index 1 Param 3=&FloatFromCurve(
     FloatCurve=DataInterface() {propertyValues {Curve {defaultValue=0, keys [{tangentMode=RCTM_User, time=0, value=1}, {tangentMode=RCTM_User, time=1, value=0}]}}}
     CurveIndex=&Frac_Float(Float=&Multiply_Float(A=$Engine.Time, B=2))
    )
    Index 2 Param 1=15
    Index 2 Param 2=3
    Index 2 Param 3=0.5
    Index 2 Param 4=0.6
    Module.Index0Float0Write=true
    Module.Index0Float1Write=true
    Module.Index0Float2Write=true
    Module.Index0Float3Write=true
    Module.Index1Float0Write=true
    Module.Index1Float1Write=true
    Module.Index1Float2Write=true
    Module.Index1Float3Write=true
    Module.Index2Float0Write=true
    Module.Index2Float1Write=true
    Module.Index2Float2Write=false
    Module.Index2Float3Write=false
    Write Parameter Index 1=::NewEnumerator1
    Write Parameter Index 2=::NewEnumerator1
   )
  ]
  rendererClasses [&NiagaraRibbonRendererProperties]
  renderers [{rendererClass=&NiagaraRibbonRendererProperties}]
 }
 TrailColor_0 {
  bEnabled=false
  emitterUpdateScript [
   &EmitterState()
   &FN_SpawnAnimTrail_v2(
    Number of Anim Trails:NiagaraInt32=&MergeHelper_IntSwitch(0=1, 1=0, 2=0, 3=?, 4=?, 5=?, Index=$System.Variant Int, Num=::NewEnumerator1)
    Spawn Interval:NiagaraInt32=&MergeHelper_IntSwitch(0=1, 1=0, 2=0, 3=?, 4=?, 5=?, Index=$System.Variant Int, Num=::NewEnumerator1)
   )
  ]
  particleSpawnScript [
   &FN_InitializeAnimTrailBoneIndecies(First Bone or Socket Index=1)
   &FN_InitializeAnimTrail(
    AnimTrailWidth=1
    Color=&MultiplyLinearColor(Linear Color=$System.TOD Color, Scale Factor=<1 1 1 0.7>)
    Lifetime=0.15
   )
   &AddVelocity(Velocity=<0 500 0>) {enabled=false}
   /SetVariables(Particles.VisibilityTag=$System.Variant Int) {isSetParameters=true, module=SetVariables_#12}
  ]
  particleUpdateScript [
   &ScaleRibbonWidth(Ribbon Width Scale=&FloatFromCurve(FloatCurve=DataInterface() {propertyValues {Curve {keys [
         {arriveTangent=-1.2579, interpMode=RCIM_Cubic, time=0, value=1}
         {arriveTangent=-0.9664, interpMode=RCIM_Cubic, leaveTangent=-0.9664, tangentMode=RCTM_User, time=1, value=0}
        ]}}})) {enabled=false}
   &ScaleColor(
    Scale Alpha=&FloatFromCurve(
     FloatCurve=DataInterface() {propertyValues {Curve {keys [
         {interpMode=RCIM_Cubic, leaveTangent=-0.5729, tangentMode=RCTM_Break, time=0, value=1}
         {arriveTangent=-0.5809, interpMode=RCIM_Cubic, leaveTangent=-0.5809, tangentMode=RCTM_User, time=0.5, value=0.15}
         {arriveTangent=-0.2139, interpMode=RCIM_Cubic, tangentMode=RCTM_Break, time=1, value=0}
        ]}}}
     Scale Curve=&FloatFromCurve(
      FloatCurve=DataInterface() {propertyValues {Curve {keys [
          <at 0 0.9 RCIM_Cubic>
          {arriveTangent=0.2255, interpMode=RCIM_Cubic, leaveTangent=0.2255, time=0.2025, value=0.95}
          {arriveTangentWeight=0.0858, interpMode=RCIM_Cubic, leaveTangentWeight=0.0558, tangentMode=RCTM_User, time=0.4434, value=1}
          {arriveTangentWeight=0.0572, interpMode=RCIM_Cubic, leaveTangentWeight=0.1462, tangentMode=RCTM_User, time=0.6, value=1}
          <at 1 0.9 RCIM_Cubic>
         ]}}}
      CurveIndex=Linked() {linkedVariable {name=NPC.FortniteNPC.TimeOfDay_0-1, type=NiagaraFloat}}
      Scale Curve=$System.AlphaFade
     )
    )
    ScaleA=true
    ScaleRGB=true
   )
   &FN_UpdateAnimTrail_v2()
   &SolveForcesAndVelocity(Clamp Velocity=false, Limit Acceleration=false)
   &DynamicMaterialParameters(
    Index 0 Param 1=0.6
    Index 1 Param 1=3
    Index 1 Param 3=&FloatFromCurve(
     FloatCurve=DataInterface() {propertyValues {Curve {defaultValue=0, keys [{tangentMode=RCTM_User, time=0, value=1}, {tangentMode=RCTM_User, time=1, value=0}]}}}
     CurveIndex=&Frac_Float(Float=&Multiply_Float(A=$Engine.Time, B=2))
    )
    Index 2 Param 1=15
    Index 2 Param 2=3
    Index 2 Param 3=0.5
    Index 2 Param 4=0.6
    Module.Index0Float0Write=true
    Module.Index0Float1Write=true
    Module.Index0Float2Write=true
    Module.Index0Float3Write=true
    Module.Index1Float0Write=true
    Module.Index1Float1Write=true
    Module.Index1Float2Write=true
    Module.Index1Float3Write=true
    Module.Index2Float0Write=true
    Module.Index2Float1Write=true
    Module.Index2Float2Write=false
    Module.Index2Float3Write=false
    Write Parameter Index 1=::NewEnumerator1
    Write Parameter Index 2=::NewEnumerator1
   )
  ]
  rendererClasses [&NiagaraRibbonRendererProperties]
  renderers [{rendererClass=&NiagaraRibbonRendererProperties}]
 }
 Trail_Color001 {
  emitterUpdateScript [
   &EmitterState()
   &FN_SpawnAnimTrail_v2(
    Number of Anim Trails:NiagaraInt32=&MergeHelper_IntSwitch(0=0, 1=1, 2=0, 3=?, 4=?, 5=?, Index=$System.Variant Int, Num=::NewEnumerator1)
    Spawn Interval:NiagaraInt32=&MergeHelper_IntSwitch(0=0, 1=1, 2=0, 3=?, 4=?, 5=?, Index=$System.Variant Int, Num=::NewEnumerator1)
   )
  ]
  particleSpawnScript [
   &FN_InitializeAnimTrailBoneIndecies(First Bone or Socket Index=1)
   &FN_InitializeAnimTrail(
    AnimTrailWidth=1
    Color=&MultiplyLinearColor(Linear Color=$System.TOD Color, Scale Factor=<1 1 1 0.55>)
    Lifetime=0.25
   )
   &AddVelocity(Velocity=<0 500 0>) {enabled=false}
   /SetVariables(Particles.VisibilityTag=$System.Variant Int) {isSetParameters=true, module=SetVariables_#12}
  ]
  particleUpdateScript [
   &ScaleRibbonWidth(Ribbon Width Scale=&FloatFromCurve(FloatCurve=DataInterface() {propertyValues {Curve {keys [
         {arriveTangent=-1.2579, interpMode=RCIM_Cubic, time=0, value=1}
         {arriveTangent=-0.9664, interpMode=RCIM_Cubic, leaveTangent=-0.9664, tangentMode=RCTM_User, time=1, value=0}
        ]}}})) {enabled=false}
   &ScaleColor(
    Scale Alpha=&FloatFromCurve(
     FloatCurve=DataInterface() {propertyValues {Curve {keys [
         {interpMode=RCIM_Cubic, leaveTangent=-0.5729, tangentMode=RCTM_Break, time=0, value=1}
         {arriveTangent=-0.5809, interpMode=RCIM_Cubic, leaveTangent=-0.5809, tangentMode=RCTM_User, time=0.5, value=0.15}
         {arriveTangent=-0.2139, interpMode=RCIM_Cubic, tangentMode=RCTM_Break, time=1, value=0}
        ]}}}
     Scale Curve=&FloatFromCurve(
      FloatCurve=DataInterface() {propertyValues {Curve {keys [
          <at 0 0.9 RCIM_Cubic>
          {arriveTangent=0.2255, interpMode=RCIM_Cubic, leaveTangent=0.2255, time=0.2025, value=0.95}
          {arriveTangentWeight=0.0858, interpMode=RCIM_Cubic, leaveTangentWeight=0.0558, tangentMode=RCTM_User, time=0.4434, value=1}
          {arriveTangentWeight=0.0572, interpMode=RCIM_Cubic, leaveTangentWeight=0.1462, tangentMode=RCTM_User, time=0.6, value=1}
          <at 1 0.9 RCIM_Cubic>
         ]}}}
      CurveIndex=Linked() {linkedVariable {name=NPC.FortniteNPC.TimeOfDay_0-1, type=NiagaraFloat}}
      Scale Curve=$System.AlphaFade
     )
    )
    ScaleA=true
    ScaleRGB=true
   )
   &FN_UpdateAnimTrail_v2()
   &SolveForcesAndVelocity(Clamp Velocity=false, Limit Acceleration=false)
   &DynamicMaterialParameters(
    Index 0 Param 1=0.6
    Index 1 Param 1=3
    Index 1 Param 3=&FloatFromCurve(
     FloatCurve=DataInterface() {propertyValues {Curve {defaultValue=0, keys [{tangentMode=RCTM_User, time=0, value=1}, {tangentMode=RCTM_User, time=1, value=0}]}}}
     CurveIndex=&Frac_Float(Float=&Multiply_Float(A=$Engine.Time, B=2))
    )
    Index 2 Param 1=15
    Index 2 Param 2=3
    Index 2 Param 3=0.5
    Index 2 Param 4=0.6
    Module.Index0Float0Write=true
    Module.Index0Float1Write=true
    Module.Index0Float2Write=true
    Module.Index0Float3Write=true
    Module.Index1Float0Write=true
    Module.Index1Float1Write=true
    Module.Index1Float2Write=true
    Module.Index1Float3Write=true
    Module.Index2Float0Write=true
    Module.Index2Float1Write=true
    Module.Index2Float2Write=false
    Module.Index2Float3Write=false
    Write Parameter Index 1=::NewEnumerator1
    Write Parameter Index 2=::NewEnumerator1
   )
  ]
  rendererClasses [&NiagaraRibbonRendererProperties]
  renderers [{rendererClass=&NiagaraRibbonRendererProperties}]
 }
 Trail_Color2 {
  bEnabled=false
  emitterUpdateScript [
   &EmitterState()
   &FN_SpawnAnimTrail_v2(
    Number of Anim Trails:NiagaraInt32=&MergeHelper_IntSwitch(0=0, 1=0, 2=1, 3=?, 4=?, 5=?, Index=$System.Variant Int, Num=::NewEnumerator1)
    Spawn Interval:NiagaraInt32=&MergeHelper_IntSwitch(0=0, 1=0, 2=1, 3=?, 4=?, 5=?, Index=$System.Variant Int, Num=::NewEnumerator1)
   )
  ]
  particleSpawnScript [
   &FN_InitializeAnimTrailBoneIndecies(First Bone or Socket Index=1)
   &FN_InitializeAnimTrail(
    AnimTrailWidth=1
    Color=&MultiplyLinearColor(Linear Color=$System.TOD Color, Scale Factor=<1 1 1 0.6>)
    Lifetime=0.15
   )
   &AddVelocity(Velocity=<0 500 0>) {enabled=false}
   /SetVariables(Particles.VisibilityTag=$System.Variant Int) {isSetParameters=true, module=SetVariables_#12}
  ]
  particleUpdateScript [
   &ScaleRibbonWidth(Ribbon Width Scale=&FloatFromCurve(FloatCurve=DataInterface() {propertyValues {Curve {keys [
         {arriveTangent=-1.2579, interpMode=RCIM_Cubic, time=0, value=1}
         {arriveTangent=-0.9664, interpMode=RCIM_Cubic, leaveTangent=-0.9664, tangentMode=RCTM_User, time=1, value=0}
        ]}}})) {enabled=false}
   &ScaleColor(
    Scale Alpha=&FloatFromCurve(
     FloatCurve=DataInterface() {propertyValues {Curve {keys [
         {interpMode=RCIM_Cubic, leaveTangent=-0.5729, tangentMode=RCTM_Break, time=0, value=1}
         {arriveTangent=-0.5809, interpMode=RCIM_Cubic, leaveTangent=-0.5809, tangentMode=RCTM_User, time=0.5, value=0.15}
         {arriveTangent=-0.2139, interpMode=RCIM_Cubic, tangentMode=RCTM_Break, time=1, value=0}
        ]}}}
     Scale Curve=&FloatFromCurve(
      FloatCurve=DataInterface() {propertyValues {Curve {keys [
          <at 0 0.9 RCIM_Cubic>
          {arriveTangent=0.2255, interpMode=RCIM_Cubic, leaveTangent=0.2255, time=0.2025, value=0.95}
          {arriveTangentWeight=0.0858, interpMode=RCIM_Cubic, leaveTangentWeight=0.0558, tangentMode=RCTM_User, time=0.4434, value=1}
          {arriveTangentWeight=0.0572, interpMode=RCIM_Cubic, leaveTangentWeight=0.1462, tangentMode=RCTM_User, time=0.6, value=1}
          <at 1 0.9 RCIM_Cubic>
         ]}}}
      CurveIndex=Linked() {linkedVariable {name=NPC.FortniteNPC.TimeOfDay_0-1, type=NiagaraFloat}}
      Scale Curve=$System.AlphaFade
     )
    )
    ScaleA=true
    ScaleRGB=true
   )
   &FN_UpdateAnimTrail_v2()
   &SolveForcesAndVelocity(Clamp Velocity=false, Limit Acceleration=false)
   &DynamicMaterialParameters(
    Index 0 Param 1=0.6
    Index 1 Param 1=3
    Index 1 Param 3=&FloatFromCurve(
     FloatCurve=DataInterface() {propertyValues {Curve {defaultValue=0, keys [{tangentMode=RCTM_User, time=0, value=1}, {tangentMode=RCTM_User, time=1, value=0}]}}}
     CurveIndex=&Frac_Float(Float=&Multiply_Float(A=$Engine.Time, B=2))
    )
    Index 2 Param 1=15
    Index 2 Param 2=3
    Index 2 Param 3=0.5
    Index 2 Param 4=0.6
    Module.Index0Float0Write=true
    Module.Index0Float1Write=true
    Module.Index0Float2Write=true
    Module.Index0Float3Write=true
    Module.Index1Float0Write=true
    Module.Index1Float1Write=true
    Module.Index1Float2Write=true
    Module.Index1Float3Write=true
    Module.Index2Float0Write=true
    Module.Index2Float1Write=true
    Module.Index2Float2Write=false
    Module.Index2Float3Write=false
    Write Parameter Index 1=::NewEnumerator1
    Write Parameter Index 2=::NewEnumerator1
   )
  ]
  rendererClasses [&NiagaraRibbonRendererProperties]
  renderers [{rendererClass=&NiagaraRibbonRendererProperties}]
 }
 Trail_Color2001 {
  emitterUpdateScript [
   &EmitterState()
   &FN_SpawnAnimTrail_v2(
    Number of Anim Trails:NiagaraInt32=&MergeHelper_IntSwitch(0=0, 1=0, 2=1, 3=?, 4=?, 5=?, Index=$System.Variant Int, Num=::NewEnumerator1)
    Spawn Interval:NiagaraInt32=&MergeHelper_IntSwitch(0=0, 1=0, 2=1, 3=?, 4=?, 5=?, Index=$System.Variant Int, Num=::NewEnumerator1)
   )
  ]
  particleSpawnScript [
   &FN_InitializeAnimTrailBoneIndecies(First Bone or Socket Index=1)
   &FN_InitializeAnimTrail(
    AnimTrailWidth=1
    Color=&MultiplyLinearColor(Linear Color=$System.TOD Color, Scale Factor=<1 1 1 0.6>)
    Lifetime=0.25
   )
   &AddVelocity(Velocity=<0 500 0>) {enabled=false}
   /SetVariables(Particles.VisibilityTag=$System.Variant Int) {isSetParameters=true, module=SetVariables_#12}
  ]
  particleUpdateScript [
   &ScaleRibbonWidth(Ribbon Width Scale=&FloatFromCurve(FloatCurve=DataInterface() {propertyValues {Curve {keys [
         {arriveTangent=-1.2579, interpMode=RCIM_Cubic, time=0, value=1}
         {arriveTangent=-0.9664, interpMode=RCIM_Cubic, leaveTangent=-0.9664, tangentMode=RCTM_User, time=1, value=0}
        ]}}})) {enabled=false}
   &ScaleColor(
    Scale Alpha=&FloatFromCurve(
     FloatCurve=DataInterface() {propertyValues {Curve {keys [
         {interpMode=RCIM_Cubic, leaveTangent=-0.5729, tangentMode=RCTM_Break, time=0, value=1}
         {arriveTangent=-0.5809, interpMode=RCIM_Cubic, leaveTangent=-0.5809, tangentMode=RCTM_User, time=0.5, value=0.15}
         {arriveTangent=-0.2139, interpMode=RCIM_Cubic, tangentMode=RCTM_Break, time=1, value=0}
        ]}}}
     Scale Curve=&FloatFromCurve(
      FloatCurve=DataInterface() {propertyValues {Curve {keys [
          <at 0 0.9 RCIM_Cubic>
          {arriveTangent=0.2255, interpMode=RCIM_Cubic, leaveTangent=0.2255, time=0.2025, value=0.95}
          {arriveTangentWeight=0.0858, interpMode=RCIM_Cubic, leaveTangentWeight=0.0558, tangentMode=RCTM_User, time=0.4434, value=1}
          {arriveTangentWeight=0.0572, interpMode=RCIM_Cubic, leaveTangentWeight=0.1462, tangentMode=RCTM_User, time=0.6, value=1}
          <at 1 0.9 RCIM_Cubic>
         ]}}}
      CurveIndex=Linked() {linkedVariable {name=NPC.FortniteNPC.TimeOfDay_0-1, type=NiagaraFloat}}
      Scale Curve=$System.AlphaFade
     )
    )
    ScaleA=true
    ScaleRGB=true
   )
   &FN_UpdateAnimTrail_v2()
   &SolveForcesAndVelocity(Clamp Velocity=false, Limit Acceleration=false)
   &DynamicMaterialParameters(
    Index 0 Param 1=0.6
    Index 1 Param 1=3
    Index 1 Param 3=&FloatFromCurve(
     FloatCurve=DataInterface() {propertyValues {Curve {defaultValue=0, keys [{tangentMode=RCTM_User, time=0, value=1}, {tangentMode=RCTM_User, time=1, value=0}]}}}
     CurveIndex=&Frac_Float(Float=&Multiply_Float(A=$Engine.Time, B=2))
    )
    Index 2 Param 1=15
    Index 2 Param 2=3
    Index 2 Param 3=0.5
    Index 2 Param 4=0.6
    Module.Index0Float0Write=true
    Module.Index0Float1Write=true
    Module.Index0Float2Write=true
    Module.Index0Float3Write=true
    Module.Index1Float0Write=true
    Module.Index1Float1Write=true
    Module.Index1Float2Write=true
    Module.Index1Float3Write=true
    Module.Index2Float0Write=true
    Module.Index2Float1Write=true
    Module.Index2Float2Write=false
    Module.Index2Float3Write=false
    Write Parameter Index 1=::NewEnumerator1
    Write Parameter Index 2=::NewEnumerator1
   )
  ]
  rendererClasses [&NiagaraRibbonRendererProperties]
  renderers [{rendererClass=&NiagaraRibbonRendererProperties}]
 }
 Trail_Color_1 {
  bEnabled=false
  emitterUpdateScript [
   &EmitterState()
   &FN_SpawnAnimTrail_v2(
    Number of Anim Trails:NiagaraInt32=&MergeHelper_IntSwitch(0=0, 1=1, 2=0, 3=?, 4=?, 5=?, Index=$System.Variant Int, Num=::NewEnumerator1)
    Spawn Interval:NiagaraInt32=&MergeHelper_IntSwitch(0=0, 1=1, 2=0, 3=?, 4=?, 5=?, Index=$System.Variant Int, Num=::NewEnumerator1)
   )
  ]
  particleSpawnScript [
   &FN_InitializeAnimTrailBoneIndecies(First Bone or Socket Index=1)
   &FN_InitializeAnimTrail(
    AnimTrailWidth=1
    Color=&MultiplyLinearColor(Linear Color=$System.TOD Color, Scale Factor=<1 1 1 0.7>)
    Lifetime=0.15
   )
   &AddVelocity(Velocity=<0 500 0>) {enabled=false}
   /SetVariables(Particles.VisibilityTag=$System.Variant Int) {isSetParameters=true, module=SetVariables_#12}
  ]
  particleUpdateScript [
   &ScaleRibbonWidth(Ribbon Width Scale=&FloatFromCurve(FloatCurve=DataInterface() {propertyValues {Curve {keys [
         {arriveTangent=-1.2579, interpMode=RCIM_Cubic, time=0, value=1}
         {arriveTangent=-0.9664, interpMode=RCIM_Cubic, leaveTangent=-0.9664, tangentMode=RCTM_User, time=1, value=0}
        ]}}})) {enabled=false}
   &ScaleColor(
    Scale Alpha=&FloatFromCurve(
     FloatCurve=DataInterface() {propertyValues {Curve {keys [
         {interpMode=RCIM_Cubic, leaveTangent=-0.5729, tangentMode=RCTM_Break, time=0, value=1}
         {arriveTangent=-0.5809, interpMode=RCIM_Cubic, leaveTangent=-0.5809, tangentMode=RCTM_User, time=0.5, value=0.15}
         {arriveTangent=-0.2139, interpMode=RCIM_Cubic, tangentMode=RCTM_Break, time=1, value=0}
        ]}}}
     Scale Curve=&FloatFromCurve(
      FloatCurve=DataInterface() {propertyValues {Curve {keys [
          <at 0 0.9 RCIM_Cubic>
          {arriveTangent=0.2255, interpMode=RCIM_Cubic, leaveTangent=0.2255, time=0.2025, value=0.95}
          {arriveTangentWeight=0.0858, interpMode=RCIM_Cubic, leaveTangentWeight=0.0558, tangentMode=RCTM_User, time=0.4434, value=1}
          {arriveTangentWeight=0.0572, interpMode=RCIM_Cubic, leaveTangentWeight=0.1462, tangentMode=RCTM_User, time=0.6, value=1}
          <at 1 0.9 RCIM_Cubic>
         ]}}}
      CurveIndex=Linked() {linkedVariable {name=NPC.FortniteNPC.TimeOfDay_0-1, type=NiagaraFloat}}
      Scale Curve=$System.AlphaFade
     )
    )
    ScaleA=true
    ScaleRGB=true
   )
   &FN_UpdateAnimTrail_v2()
   &SolveForcesAndVelocity(Clamp Velocity=false, Limit Acceleration=false)
   &DynamicMaterialParameters(
    Index 0 Param 1=0.6
    Index 1 Param 1=3
    Index 1 Param 3=&FloatFromCurve(
     FloatCurve=DataInterface() {propertyValues {Curve {defaultValue=0, keys [{tangentMode=RCTM_User, time=0, value=1}, {tangentMode=RCTM_User, time=1, value=0}]}}}
     CurveIndex=&Frac_Float(Float=&Multiply_Float(A=$Engine.Time, B=2))
    )
    Index 2 Param 1=15
    Index 2 Param 2=3
    Index 2 Param 3=0.5
    Index 2 Param 4=0.6
    Module.Index0Float0Write=true
    Module.Index0Float1Write=true
    Module.Index0Float2Write=true
    Module.Index0Float3Write=true
    Module.Index1Float0Write=true
    Module.Index1Float1Write=true
    Module.Index1Float2Write=true
    Module.Index1Float3Write=true
    Module.Index2Float0Write=true
    Module.Index2Float1Write=true
    Module.Index2Float2Write=false
    Module.Index2Float3Write=false
    Write Parameter Index 1=::NewEnumerator1
    Write Parameter Index 2=::NewEnumerator1
   )
  ]
  rendererClasses [&NiagaraRibbonRendererProperties]
  renderers [{rendererClass=&NiagaraRibbonRendererProperties}]
 }
 VBucks {
  emitterUpdateScript [
   &EmitterState(Life Cycle Mode=::NewEnumerator1, Loop Behavior=::NewEnumerator1, Loop Duration=1, UseLoopDelay=false)
   &SpawnRate(
    Spawn Probability=&FloatFromCurve(
     FloatCurve=DataInterface() {propertyValues {Curve {
        defaultValue=1
        keys [
         {arriveTangent=7.8977, interpMode=RCIM_Cubic, leaveTangent=7.8977, leaveTangentWeight=0.3558, tangentMode=RCTM_User, time=0, value=0}
         {arriveTangentWeight=0.2236, interpMode=RCIM_Cubic, leaveTangentWeight=0.1491, tangentMode=RCTM_User, time=0.3, value=1}
         {arriveTangentWeight=0.0667, interpMode=RCIM_Cubic, leaveTangentWeight=0.1667, tangentMode=RCTM_User, time=0.6, value=1}
         {arriveTangent=-5.9246, tangentMode=RCTM_User, time=1, value=0}
        ]
       }}}
     CurveIndex=$Emitter.NormalizedLoopAge
    )
    SpawnRate=&RandomRangeFloat(Minimum=5, Maximum=12, Evaluation Type=::NewEnumerator1)
    Use Spawn Probability=true
   )
  ]
  particleSpawnScript [
   &PartitionParticles(Number Of Partitions=3)
   &InitializeParticle(
    Color=&MergeHelper_ColorSwitch(
     0=&SelectLinearColorFromArray(
      Array Sampling Mode=::NewEnumerator1
      Direct Array Index=$Particles.PartitionParticles.Partition
      Color Selection Array=DataInterface() {propertyValues {ColorData [<rgba 0.36 0.9 0.5535 1>, <rgba 1 0.4453 0.619 1>]}}
      Random Seed=?
     )
     1=<rgba 1 0.5467 0.15 1>
     2=<rgba 1 0.5467 0.15 1>
     3=?
     Index=$System.Variant Int
     Num=::NewEnumerator1
    )
    Lifetime=&RandomRangeFloat(Minimum=0.5, Maximum=0.7)
    Mesh Scale Mode=::NewEnumerator3
    Mesh Uniform Scale=0.8
    Sprite Size Mode=::NewEnumerator3
    UsePositionOffset=false
    Write Material Random=true
   )
   &InitialMeshOrientation(
    Mesh Orientation Mode=::NewEnumerator2
    Orientation Coordinate Space=::Local
    Orientation Vector=$Particles.Velocity
    Rotation Coordinate Space=::NewEnumerator2
    Use Rotation Vector=true
   ) {module=InitialMeshOrientation001}
   &InitialMeshOrientation(Use Rotation Vector=false) {enabled=false}
   &SocketLocation(
    Interpolate Socket Sampling=true
    SamplePosOffset=true
    Sampled Position Offset=<0 0 15>
    Socket Index=1
    Socket Reader=$System.Socket Reader
    Socket Sampling Mode=::NewEnumerator3
    VelocitySpeedLimit=false
   )
   &ShapeLocation(MaskSpawnGroup=false, Override Seed=false, Sphere Radius=10, UseOwnerScale=true)
   &InheritVelocity(Inherited Velocity Amount Scale=<0.08 0.08 0.08>, Inherited Velocity Speed Limit=50, Speed Limit=true, UseSpeedThreshold=false)
   &AddVelocity(
    Rotation Coordinate Space=::World
    Velocity=$Output.SocketLocation.SampledVelocity
    Velocity Speed Scale=&RandomRangeFloat(Minimum=0.065, Maximum=0.07)
   )
   &CameraOffset(Camera Offset Amount=20) {enabled=false}
   /SetVariables(Particles.MeshIndex=&RandomRangeInt(Minimum=0, Maximum=2)) {isSetParameters=true, module=SetVariables_#13}
   /SetVariables(Particles.VisibilityTag=&MergeHelper_IntSwitch(0=0, 1=1, 2=1, 3=?, 4=?, 5=?, Index=$System.Variant Int, Num=::NewEnumerator1)) {isSetParameters=true, module=SetVariables_#14}
   /SetVariables(Particles.DistanceScaleOpacity=1) {isSetParameters=true, module=SetVariables_#15}
  ]
  particleUpdateScript [
   &ParticleState()
   &ScaleSpriteSize(
    Scale Factor=&Multiply_Vector2D_ByFloat(
     Float=&FloatFromCurve(FloatCurve=DataInterface() {propertyValues {Curve {
         defaultValue=1
         keys [
          {arriveTangent=0.0844, interpMode=RCIM_Cubic, leaveTangent=0.0844, leaveTangentWeight=0.3333, tangentMode=RCTM_User, time=0, value=0.53}
          {
           arriveTangent=0.3354
           arriveTangentWeight=0.1911
           interpMode=RCIM_Cubic
           leaveTangent=0.3354
           leaveTangentWeight=0.1422
           tangentMode=RCTM_User
           time=0.5734
           value=0.5966
          }
          {
           arriveTangent=0.7849
           arriveTangentWeight=0.0271
           interpMode=RCIM_Cubic
           leaveTangent=0.7849
           leaveTangentWeight=0.1151
           tangentMode=RCTM_User
           time=0.6546
           value=0.6043
          }
          {
           arriveTangent=0.6802
           arriveTangentWeight=0.0306
           interpMode=RCIM_Cubic
           leaveTangent=0.6802
           leaveTangentWeight=0.0846
           tangentMode=RCTM_User
           time=0.7463
           value=0.6851
          }
          {
           arriveTangent=0.4774
           arriveTangentWeight=0.0394
           interpMode=RCIM_Cubic
           leaveTangent=0.4774
           leaveTangentWeight=0.0452
           tangentMode=RCTM_User
           time=0.8644
           value=0.6515
          }
          {
           arriveTangent=12.3514
           arriveTangentWeight=0.0119
           interpMode=RCIM_Cubic
           leaveTangent=12.3514
           leaveTangentWeight=0.0333
           tangentMode=RCTM_User
           time=0.9
           value=1
          }
          {interpMode=RCIM_Cubic, tangentMode=RCTM_User, time=1, value=1}
         ]
        }}})
     Vector2D=&MakeVector2D(
      X=&FloatFromCurve(FloatCurve=DataInterface() {propertyValues {Curve {keys [
           {interpMode=RCIM_Cubic, leaveTangentWeight=0.2737, tangentMode=RCTM_User, time=0, value=1.35}
           {arriveTangent=-4.0805, interpMode=RCIM_Cubic, leaveTangent=-4.0805, tangentMode=RCTM_User, time=0.1813, value=1}
           {arriveTangent=-0.1492, interpMode=RCIM_Cubic, leaveTangent=-0.1492, time=0.2013, value=0.8992}
           <at 0.2613 0.9881 RCIM_Cubic>
          ]}}})
      Y=&FloatFromCurve(FloatCurve=DataInterface() {propertyValues {Curve {keys [
           {arriveTangent=-0.2248, interpMode=RCIM_Cubic, leaveTangent=-0.2248, tangentMode=RCTM_User, time=0, value=0.75}
           {arriveTangent=3.3994, interpMode=RCIM_Cubic, leaveTangent=3.3994, tangentMode=RCTM_User, time=0.1813, value=1}
           <at 0.2013 1.1617 RCIM_Cubic>
           <at 0.2613 1 RCIM_Cubic>
          ]}}})
     )
    )
    Scale Sprite Size Mode=::NewEnumerator2
   )
   /SetVariables(Particles.DistanceScaleOpacity:NiagaraFloat=/Niagara/DynamicInputs/Math/Min_Float(
     A:NiagaraFloat=$Particles.DistanceScaleOpacity
     B:NiagaraFloat=/Game/Effects/Niagara/DynamicInputs/Math/FN_Fit(
      Clamp:NiagaraBool=true
      Value:NiagaraFloat=&FloatFromCurve(
       FloatCurve=DataInterface() {propertyValues {Curve {keys [
           {interpMode=RCIM_Cubic, leaveTangentWeight=0.4667, tangentMode=RCTM_User, time=0.8, value=1}
           {arriveTangent=-2.7409, interpMode=RCIM_Cubic, leaveTangent=-2.7409, tangentMode=RCTM_User, time=2, value=0}
          ]}}}
       CurveIndex=&Divide_Float(
        A=&VectorLength(Vector=&SubtractVectorFromPosition(Position=$Particles.Position, Vector=Linked() {linkedVariable {name=Engine.Emitter.SimulationPosition, type=NiagaraPosition}}))
        B=125
       )
      )
      SourceMin:NiagaraFloat=0
      SourceMax:NiagaraFloat=1
      DestMin:NiagaraFloat=0
      DestMax:NiagaraFloat=1
     )
    )) {isSetParameters=true, module=SetVariables_#16}
   &UpdateMeshOrientation(
    Rotation Rate=&FloatFromCurve(
     FloatCurve=DataInterface() {propertyValues {Curve {
        defaultValue=1
        keys [
         {arriveTangent=0.1962, interpMode=RCIM_Cubic, leaveTangent=0.1962, tangentMode=RCTM_User, time=0, value=0}
         {interpMode=RCIM_Cubic, tangentMode=RCTM_User, time=0.9, value=1}
        ]
       }}}
     Scale Curve=&RandomRangeFloat(Minimum=1, Maximum=2)
    )
    Rotation Vector=&RandomRangeVector(Minimum=<-1 0 0>, Maximum=<1 0.2 0.2>)
   )
   &Drag(
    Drag=&FloatFromCurve(
     FloatCurve=DataInterface() {propertyValues {Curve {keys [
         {arriveTangent=3.9287, interpMode=RCIM_Cubic, leaveTangent=3.9287, tangentMode=RCTM_User, time=0, value=0.3}
         {
          arriveTangent=-2.2912
          arriveTangentWeight=0.0483
          interpMode=RCIM_Cubic
          leaveTangent=-2.2912
          leaveTangentWeight=0.1448
          tangentMode=RCTM_User
          time=0.15
          value=1
         }
         <at 1 0.22 RCIM_Cubic>
        ]}}}
     Scale Curve=&RandomRangeFloat(Minimum=5, Maximum=8)
    )
    Use Linear Drag=true
    Use Rotational Drag=true
   ) {enabled=false}
   &GravityForce(Gravity=&RandomRangeVector(Minimum=<0 0 -600>, Maximum=<1 1 -980>))
   &ScaleMeshSize(Scale Factor=&Multiply_VectorByFloat(
     Vector=&VectorFromCurve(
      VectorCurve=DataInterface() {propertyValues {XCurve {keys [<at 0 1>,<at 1 1>]}, YCurve {keys [<at 0 1>,<at 1 1>]}, ZCurve {keys [<at 0 0.1>,<at 0.4 1>]}}}
      CurveIndex=$Particles.Age
     )
     Float=&Multiply_Float(
      A=&FloatFromCurve(
       FloatCurve=DataInterface() {propertyValues {Curve {keys [
           {arriveTangent=4.1785, interpMode=RCIM_Cubic, leaveTangent=8.7018, tangentMode=RCTM_Break, time=0, value=0}
           {arriveTangent=0.0224, interpMode=RCIM_Cubic, leaveTangent=0.0224, tangentMode=RCTM_Break, time=0.2, value=1}
          ]}}}
       CurveIndex=$Particles.Age
       Scale Curve=&FloatFromCurve(FloatCurve=DataInterface() {propertyValues {Curve {keys [
            {interpMode=RCIM_Cubic, tangentMode=RCTM_Break, time=0.7, value=1}
            {arriveTangent=-6.5758, arriveTangentWeight=0.855, interpMode=RCIM_Cubic, tangentMode=RCTM_Break, time=1, value=0}
           ]}}})
      )
      B=$Particles.DistanceScaleOpacity
     )
    )) {module=ScaleMeshSize001}
   &CurlNoiseForce(
    Bias Noise=false
    Noise Strength=&FloatFromCurve(
     FloatCurve=DataInterface() {propertyValues {Curve {
        defaultValue=1
        keys [
         {arriveTangent=0.7445, interpMode=RCIM_Cubic, leaveTangent=0.7445, tangentMode=RCTM_User, time=0, value=0}
         {interpMode=RCIM_Cubic, tangentMode=RCTM_User, time=0.5, value=1}
        ]
       }}}
     Scale Curve=500
    )
    Pan Noise=false
   )
   &AerodynamicDrag(
    Aerodynamic Drag:NiagaraFloat=&FloatFromCurve(
     FloatCurve=DataInterface() {propertyValues {Curve {keys [
         {arriveTangent=4.1123, interpMode=RCIM_Cubic, leaveTangent=4.1123, leaveTangentWeight=0.1866, tangentMode=RCTM_User, time=0, value=0.2}
         {
          arriveTangent=-2.4122
          arriveTangentWeight=0.0483
          interpMode=RCIM_Cubic
          leaveTangent=-2.4122
          leaveTangentWeight=0.1448
          tangentMode=RCTM_User
          time=0.22
          value=1
         }
         {arriveTangentWeight=0.6217, interpMode=RCIM_Cubic, tangentMode=RCTM_User, time=1, value=0.22}
        ]}}}
     Scale Curve=2
    )
    Drag Shape Curve=DataInterface() {propertyValues {Curve {keys [
        {interpMode=RCIM_Cubic, tangentMode=RCTM_User, time=0, value=0.1}
        {arriveTangent=2.0799, interpMode=RCIM_Cubic, leaveTangent=2.0799, tangentMode=RCTM_User, time=1, value=1}
       ]}}}
    Lift Shape Curve=DataInterface() {propertyValues {Curve {keys [
        {interpMode=RCIM_Cubic, tangentMode=RCTM_User, time=0, value=0}
        {interpMode=RCIM_Cubic, tangentMode=RCTM_User, time=0.5, value=1}
        {tangentMode=RCTM_User, time=1, value=0}
       ]}}}
    Use Linear Drag=true
    Use Rotational Drag=true
   )
   &SolveForcesAndVelocity(Clamp Velocity=false, Limit Acceleration=false)
   &DynamicMaterialParameters(
    Index 0 Param 1=&MakeCustomFloatFromBool(Bool=$User.bIsFrontEndPreview, True Float=0.02, False Float=0)
    Index 0 Param 2=&FloatFromCurve(FloatCurve=DataInterface() {propertyValues {Curve {
        defaultValue=1
        keys [
         {interpMode=RCIM_Cubic, leaveTangentWeight=0.9127, tangentMode=RCTM_User, time=0.2, value=0}
         {arriveTangentWeight=0.2716, interpMode=RCIM_Cubic, leaveTangentWeight=0.4526, tangentMode=RCTM_User, time=0.5, value=0.2}
         {arriveTangent=-0.0405, interpMode=RCIM_Cubic, leaveTangent=-0.0405, tangentMode=RCTM_User, time=1, value=0}
        ]
       }}})
    Module.Index0Float0Write=true
    Module.Index0Float1Write=true
    Module.Index0Float2Write=true
    Module.Index0Float3Write=true
   )
   &AlignSpriteToMeshOrientation(SetAlignment=true, SetFacing=true)
   &KillParticles(Kill Particles=&SetBoolByFloatComparison(A=$Particles.DistanceScaleOpacity, B=0.001, Comparison Type=::NewEnumerator7))
  ]
  rendererClasses [&NiagaraMeshRendererProperties]
  renderers [
   {rendererClass=&NiagaraMeshRendererProperties, rendererIndex=0}
   {rendererClass=&NiagaraMeshRendererProperties, rendererIndex=1}
  ]
 }
 VBucks001 {
  bEnabled=false
  emitterUpdateScript [
   &EmitterState(Life Cycle Mode=::NewEnumerator1, Loop Behavior=::NewEnumerator1, Loop Duration=1, UseLoopDelay=false)
   &SpawnRate(
    Spawn Probability=&FloatFromCurve(
     FloatCurve=DataInterface() {propertyValues {Curve {
        defaultValue=1
        keys [
         {arriveTangent=7.8977, interpMode=RCIM_Cubic, leaveTangent=7.8977, leaveTangentWeight=0.3558, tangentMode=RCTM_User, time=0, value=0}
         {arriveTangentWeight=0.2236, interpMode=RCIM_Cubic, leaveTangentWeight=0.1491, tangentMode=RCTM_User, time=0.0571, value=1}
         {arriveTangentWeight=0.0667, interpMode=RCIM_Cubic, leaveTangentWeight=0.1667, tangentMode=RCTM_User, time=0.1386, value=1}
         {arriveTangent=-5.9246, tangentMode=RCTM_User, time=0.2, value=0}
        ]
       }}}
     CurveIndex=$Emitter.Age
    )
    SpawnRate=&RandomRangeFloat(Minimum=20, Maximum=40, Evaluation Type=::NewEnumerator1)
    Use Spawn Probability=true
   )
  ]
  particleSpawnScript [
   &PartitionParticles(Number Of Partitions=3)
   &InitializeParticle(
    Color=&MergeHelper_ColorSwitch(
     0=&SelectLinearColorFromArray(
      Array Sampling Mode=::NewEnumerator1
      Direct Array Index=$Particles.PartitionParticles.Partition
      Color Selection Array=DataInterface() {propertyValues {ColorData [<rgba 0.36 0.9 0.5535 1>, <rgba 1 0.4453 0.619 1>]}}
      Random Seed=?
     )
     1=<rgba 1 0.5467 0.15 1>
     2=<rgba 1 0.5467 0.15 1>
     3=?
     Index=$System.Variant Int
     Num=::NewEnumerator1
    )
    Lifetime=&RandomRangeFloat(Minimum=0.3, Maximum=0.6)
    Mesh Scale Mode=::NewEnumerator3
    Mesh Uniform Scale=0.8
    Sprite Size Mode=::NewEnumerator3
    UsePositionOffset=false
    Write Material Random=true
   )
   &InitialMeshOrientation(
    Mesh Orientation Mode=::NewEnumerator2
    Orientation Coordinate Space=::Local
    Orientation Vector=$Particles.Velocity
    Rotation Coordinate Space=::NewEnumerator2
    Use Rotation Vector=true
   ) {module=InitialMeshOrientation001}
   &InitialMeshOrientation(Use Rotation Vector=false) {enabled=false}
   &SocketLocation(
    Interpolate Socket Sampling=true
    SamplePosOffset=true
    Sampled Position Offset=<0 0 15>
    Socket Index=1
    Socket Reader=$System.Socket Reader
    Socket Sampling Mode=::NewEnumerator3
    VelocitySpeedLimit=false
   )
   &ShapeLocation(MaskSpawnGroup=false, Override Seed=false, Sphere Radius=10, UseOwnerScale=true)
   &InheritVelocity(Inherited Velocity Amount Scale=<0.08 0.08 0.08>, Inherited Velocity Speed Limit=50, Speed Limit=true, UseSpeedThreshold=false)
   &AddVelocity(
    Rotation Coordinate Space=::World
    Velocity=$Output.SocketLocation.SampledVelocity
    Velocity Speed Scale=&RandomRangeFloat(Minimum=0.065, Maximum=0.07)
   )
   &CameraOffset(Camera Offset Amount=20) {enabled=false}
   /SetVariables(Particles.MeshIndex=&RandomRangeInt(Minimum=0, Maximum=2)) {isSetParameters=true, module=SetVariables_#13}
   /SetVariables(Particles.VisibilityTag=&MergeHelper_IntSwitch(0=0, 1=1, 2=1, 3=?, 4=?, 5=?, Index=$System.Variant Int, Num=::NewEnumerator1)) {isSetParameters=true, module=SetVariables_#14}
   /SetVariables(Particles.DistanceScaleOpacity=1) {isSetParameters=true, module=SetVariables_#15}
  ]
  particleUpdateScript [
   &ParticleState()
   &ScaleSpriteSize(
    Scale Factor=&Multiply_Vector2D_ByFloat(
     Float=&FloatFromCurve(FloatCurve=DataInterface() {propertyValues {Curve {
         defaultValue=1
         keys [
          {arriveTangent=0.0844, interpMode=RCIM_Cubic, leaveTangent=0.0844, leaveTangentWeight=0.3333, tangentMode=RCTM_User, time=0, value=0.53}
          {
           arriveTangent=0.3354
           arriveTangentWeight=0.1911
           interpMode=RCIM_Cubic
           leaveTangent=0.3354
           leaveTangentWeight=0.1422
           tangentMode=RCTM_User
           time=0.5734
           value=0.5966
          }
          {
           arriveTangent=0.7849
           arriveTangentWeight=0.0271
           interpMode=RCIM_Cubic
           leaveTangent=0.7849
           leaveTangentWeight=0.1151
           tangentMode=RCTM_User
           time=0.6546
           value=0.6043
          }
          {
           arriveTangent=0.6802
           arriveTangentWeight=0.0306
           interpMode=RCIM_Cubic
           leaveTangent=0.6802
           leaveTangentWeight=0.0846
           tangentMode=RCTM_User
           time=0.7463
           value=0.6851
          }
          {
           arriveTangent=0.4774
           arriveTangentWeight=0.0394
           interpMode=RCIM_Cubic
           leaveTangent=0.4774
           leaveTangentWeight=0.0452
           tangentMode=RCTM_User
           time=0.8644
           value=0.6515
          }
          {
           arriveTangent=12.3514
           arriveTangentWeight=0.0119
           interpMode=RCIM_Cubic
           leaveTangent=12.3514
           leaveTangentWeight=0.0333
           tangentMode=RCTM_User
           time=0.9
           value=1
          }
          {interpMode=RCIM_Cubic, tangentMode=RCTM_User, time=1, value=1}
         ]
        }}})
     Vector2D=&MakeVector2D(
      X=&FloatFromCurve(FloatCurve=DataInterface() {propertyValues {Curve {keys [
           {interpMode=RCIM_Cubic, leaveTangentWeight=0.2737, tangentMode=RCTM_User, time=0, value=1.35}
           {arriveTangent=-4.0805, interpMode=RCIM_Cubic, leaveTangent=-4.0805, tangentMode=RCTM_User, time=0.1813, value=1}
           {arriveTangent=-0.1492, interpMode=RCIM_Cubic, leaveTangent=-0.1492, time=0.2013, value=0.8992}
           <at 0.2613 0.9881 RCIM_Cubic>
          ]}}})
      Y=&FloatFromCurve(FloatCurve=DataInterface() {propertyValues {Curve {keys [
           {arriveTangent=-0.2248, interpMode=RCIM_Cubic, leaveTangent=-0.2248, tangentMode=RCTM_User, time=0, value=0.75}
           {arriveTangent=3.3994, interpMode=RCIM_Cubic, leaveTangent=3.3994, tangentMode=RCTM_User, time=0.1813, value=1}
           <at 0.2013 1.1617 RCIM_Cubic>
           <at 0.2613 1 RCIM_Cubic>
          ]}}})
     )
    )
    Scale Sprite Size Mode=::NewEnumerator2
   )
   /SetVariables(Particles.DistanceScaleOpacity:NiagaraFloat=&Min_Float(
     A:NiagaraFloat=$Particles.DistanceScaleOpacity
     B:NiagaraFloat=&FN_Fit(
      Clamp:NiagaraBool=true
      Value:NiagaraFloat=&FloatFromCurve(
       FloatCurve=DataInterface() {propertyValues {Curve {keys [
           {interpMode=RCIM_Cubic, leaveTangentWeight=0.4667, tangentMode=RCTM_User, time=0.8, value=1}
           {arriveTangent=-2.7409, interpMode=RCIM_Cubic, leaveTangent=-2.7409, tangentMode=RCTM_User, time=2, value=0}
          ]}}}
       CurveIndex=&Divide_Float(
        A=&VectorLength(Vector=&SubtractVectorFromPosition(Position=$Particles.Position, Vector=Linked() {linkedVariable {name=Engine.Emitter.SimulationPosition, type=NiagaraPosition}}))
        B=125
       )
      )
      SourceMin:NiagaraFloat=0
      SourceMax:NiagaraFloat=1
      DestMin:NiagaraFloat=0
      DestMax:NiagaraFloat=1
     )
    )) {isSetParameters=true, module=SetVariables_#16}
   &UpdateMeshOrientation(
    Rotation Rate=&FloatFromCurve(
     FloatCurve=DataInterface() {propertyValues {Curve {
        defaultValue=1
        keys [
         {arriveTangent=0.1962, interpMode=RCIM_Cubic, leaveTangent=0.1962, tangentMode=RCTM_User, time=0, value=0}
         {interpMode=RCIM_Cubic, tangentMode=RCTM_User, time=0.9, value=1}
        ]
       }}}
     Scale Curve=&RandomRangeFloat(Minimum=1, Maximum=2)
    )
    Rotation Vector=&RandomRangeVector(Minimum=<-1 0 0>, Maximum=<1 0.2 0.2>)
   )
   &Drag(
    Drag=&FloatFromCurve(
     FloatCurve=DataInterface() {propertyValues {Curve {keys [
         {arriveTangent=3.9287, interpMode=RCIM_Cubic, leaveTangent=3.9287, tangentMode=RCTM_User, time=0, value=0.3}
         {
          arriveTangent=-2.2912
          arriveTangentWeight=0.0483
          interpMode=RCIM_Cubic
          leaveTangent=-2.2912
          leaveTangentWeight=0.1448
          tangentMode=RCTM_User
          time=0.15
          value=1
         }
         <at 1 0.22 RCIM_Cubic>
        ]}}}
     Scale Curve=&RandomRangeFloat(Minimum=5, Maximum=8)
    )
    Use Linear Drag=true
    Use Rotational Drag=true
   ) {enabled=false}
   &GravityForce(Gravity=&RandomRangeVector(Minimum=<0 0 -600>, Maximum=<1 1 -980>))
   &ScaleMeshSize(Scale Factor=&Multiply_VectorByFloat(
     Vector=&VectorFromCurve(
      VectorCurve=DataInterface() {propertyValues {XCurve {keys [<at 0 1>,<at 1 1>]}, YCurve {keys [<at 0 1>,<at 1 1>]}, ZCurve {keys [<at 0 0.1>,<at 0.4 1>]}}}
      CurveIndex=$Particles.Age
     )
     Float=&Multiply_Float(
      A=&FloatFromCurve(
       FloatCurve=DataInterface() {propertyValues {Curve {keys [
           {arriveTangent=4.1785, interpMode=RCIM_Cubic, leaveTangent=8.7018, tangentMode=RCTM_Break, time=0, value=0}
           {arriveTangent=0.0224, interpMode=RCIM_Cubic, leaveTangent=0.0224, tangentMode=RCTM_Break, time=0.2, value=1}
          ]}}}
       CurveIndex=$Particles.Age
       Scale Curve=&FloatFromCurve(FloatCurve=DataInterface() {propertyValues {Curve {keys [
            {interpMode=RCIM_Cubic, tangentMode=RCTM_Break, time=0.7, value=1}
            {arriveTangent=-6.5758, arriveTangentWeight=0.855, interpMode=RCIM_Cubic, tangentMode=RCTM_Break, time=1, value=0}
           ]}}})
      )
      B=$Particles.DistanceScaleOpacity
     )
    )) {module=ScaleMeshSize001}
   &CurlNoiseForce(
    Bias Noise=false
    Noise Strength=&FloatFromCurve(
     FloatCurve=DataInterface() {propertyValues {Curve {
        defaultValue=1
        keys [
         {arriveTangent=0.7445, interpMode=RCIM_Cubic, leaveTangent=0.7445, tangentMode=RCTM_User, time=0, value=0}
         {interpMode=RCIM_Cubic, tangentMode=RCTM_User, time=0.5, value=1}
        ]
       }}}
     Scale Curve=500
    )
    Pan Noise=false
   )
   &AerodynamicDrag(
    Aerodynamic Drag:NiagaraFloat=&FloatFromCurve(
     FloatCurve=DataInterface() {propertyValues {Curve {keys [
         {arriveTangent=4.1123, interpMode=RCIM_Cubic, leaveTangent=4.1123, leaveTangentWeight=0.1866, tangentMode=RCTM_User, time=0, value=0.2}
         {
          arriveTangent=-2.4122
          arriveTangentWeight=0.0483
          interpMode=RCIM_Cubic
          leaveTangent=-2.4122
          leaveTangentWeight=0.1448
          tangentMode=RCTM_User
          time=0.22
          value=1
         }
         {arriveTangentWeight=0.6217, interpMode=RCIM_Cubic, tangentMode=RCTM_User, time=1, value=0.22}
        ]}}}
     Scale Curve=2
    )
    Drag Shape Curve=DataInterface() {propertyValues {Curve {keys [
        {interpMode=RCIM_Cubic, tangentMode=RCTM_User, time=0, value=0.1}
        {arriveTangent=2.0799, interpMode=RCIM_Cubic, leaveTangent=2.0799, tangentMode=RCTM_User, time=1, value=1}
       ]}}}
    Lift Shape Curve=DataInterface() {propertyValues {Curve {keys [
        {interpMode=RCIM_Cubic, tangentMode=RCTM_User, time=0, value=0}
        {interpMode=RCIM_Cubic, tangentMode=RCTM_User, time=0.5, value=1}
        {tangentMode=RCTM_User, time=1, value=0}
       ]}}}
    Use Linear Drag=true
    Use Rotational Drag=true
   )
   &SolveForcesAndVelocity(Clamp Velocity=false, Limit Acceleration=false)
   &DynamicMaterialParameters(
    Index 0 Param 1=&MakeCustomFloatFromBool(Bool=$User.bIsFrontEndPreview, True Float=0.02, False Float=0)
    Index 0 Param 2=&FloatFromCurve(FloatCurve=DataInterface() {propertyValues {Curve {
        defaultValue=1
        keys [
         {interpMode=RCIM_Cubic, leaveTangentWeight=0.9127, tangentMode=RCTM_User, time=0.2, value=0}
         {arriveTangentWeight=0.2716, interpMode=RCIM_Cubic, leaveTangentWeight=0.4526, tangentMode=RCTM_User, time=0.5, value=0.2}
         {arriveTangent=-0.0405, interpMode=RCIM_Cubic, leaveTangent=-0.0405, tangentMode=RCTM_User, time=1, value=0}
        ]
       }}})
    Module.Index0Float0Write=true
    Module.Index0Float1Write=true
    Module.Index0Float2Write=true
    Module.Index0Float3Write=true
   )
   &AlignSpriteToMeshOrientation(SetAlignment=true, SetFacing=true)
   &KillParticles(Kill Particles=&SetBoolByFloatComparison(A=$Particles.DistanceScaleOpacity, B=0.001, Comparison Type=::NewEnumerator7))
  ]
  rendererClasses [&NiagaraMeshRendererProperties]
  renderers [
   {rendererClass=&NiagaraMeshRendererProperties, rendererIndex=0}
   {rendererClass=&NiagaraMeshRendererProperties, rendererIndex=1}
  ]
 }
}
summary {
 systemName=NS_Pickaxe_Keytar_FunkyFeather_Cap_SwingTrail
 emitters [
  {bEnabled=false, emitterName=TrailColor_0, rendererClasses [&NiagaraRibbonRendererProperties]}
  {bEnabled=false, emitterName=DEBUG_VALUES, rendererClasses [&NiagaraSpriteRendererProperties]}
  {emitterName=VBucks, rendererClasses [&NiagaraMeshRendererProperties]}
  {bEnabled=false, emitterName=Trail_Color_1, rendererClasses [&NiagaraRibbonRendererProperties]}
  {bEnabled=false, emitterName=Trail_Color2, rendererClasses [&NiagaraRibbonRendererProperties]}
  {emitterName=Flares_Sparkles, rendererClasses [&NiagaraSpriteRendererProperties]}
  {emitterName=Accent_0_1, rendererClasses [&NiagaraRibbonRendererProperties]}
  {emitterName=Accent, rendererClasses [&NiagaraRibbonRendererProperties]}
  {emitterName=TrailColor001, rendererClasses [&NiagaraRibbonRendererProperties]}
  {emitterName=Trail_Color001, rendererClasses [&NiagaraRibbonRendererProperties]}
  {emitterName=Trail_Color2001, rendererClasses [&NiagaraRibbonRendererProperties]}
  {bEnabled=false, emitterName=VBucks001, rendererClasses [&NiagaraMeshRendererProperties]}
 ]
 userVariables [
  User.bIsFrontEnd {type=NiagaraBool}
  User.bIsFrontEndPreview {type=NiagaraBool}
  User.Landing {type=NiagaraFloat, defaultValue=0, description=""}
  User.RibbonIDOffset {type=NiagaraInt32}
  User.SkeletalMesh {
   type=NiagaraDataInterfaceSkeletalMesh
   defaultValue {
    struct=NiagaraExt_VariableValue_DataInterface
    value {dataInterface=&self:NiagaraDataInterfaceSkeletalMesh_0}
   }
  }
  User.SwingSectionIndex {type=NiagaraInt32}
  User.Variant {type=NiagaraFloat, defaultValue=0, description=""}
 ]
}
