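# Animation Blueprint for the FunkyFeather keytar instrument cosmetic, adding secondary motion to its money bones and handle.
# The input pose gets additive offsets on a top money bone and the handle, then passes through chained spring-based Anim Dynamics nodes with angular limits and gravity overrides.
# The event graph sets bill alpha, handle alpha and gravity variables from front-end and emote state, but the anim graph reads none of them.
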AnimBlueprint /SparksCosmetics/Instrument/Keytar/FunkyFeather_Cap/Meshes/FunkyFeather_Keytar_AnimBP
target_skeleton=&/FunkyFeather_Keytar_Skeleton
graphs [
 AnimGraph {nodes [
   n1 = Input Pose()
   n2 = "Transform (Modify) Bone - Bone: dyn_handle_01"!(
    ComponentPose="Transform (Modify) Bone - Bone: base_Money_Top_C"!(ComponentPose=Local To Component(LocalPose=@n1), Alpha=1, Translation="0.2,-1,0") -> (Pose) {translationMode=BMM_Additive, boneToModify {boneName=base_Money_Top_C}, translation=<0.2 -1>}
    Alpha=1
    Rotation="0,-78,0"
   ) -> (Pose) {
    rotationMode=BMM_Additive
    rotationSpace=BCS_BoneSpace
    scaleMode=BMM_Additive
    translationMode=BMM_Additive
    boneToModify {boneName=dyn_handle_01}
    rotation {yaw=-78}
   }
   n3 = Anim Dynamics - Bone: dyn_Money_Top_A_01!(ComponentPose=@n2, Alpha=1) -> (Pose) {
    angularSpringConstant=15000
    bAngularSpring=true
    bUseGravityOverride=true
    simulationSpace=RootRelative
    boundBone {boneName=dyn_Money_Top_A_01}
    componentAppliedLinearAccClamp=<3000 3000 3000>
    componentLinearAccScale=<0.7 0.7 0.7>
    componentLinearVelScale=<0.5 0.5 0.5>
    gravityOverride {z=-980}
    physicsBodyDefinitions [{
      boundBone {boneName=dyn_Money_Top_A_01}
      boxExtents=<20 20 20>
      constraintSetup {angularLimitsMax=<0 5 15>, angularLimitsMin=<0 -5 0>, angularTarget=<0 0 0>}
      localJointOffset=<2 0 0>
     }]
    simSpaceSettings {simSpaceAngularAlpha=1}
   }
   n4 = Anim Dynamics - Bone: dyn_Money_Top_A_02!(ComponentPose=@n3, Alpha=1) -> (Pose) {
    angularSpringConstant=15000
    bAngularSpring=true
    bUseGravityOverride=true
    simulationSpace=RootRelative
    boundBone {boneName=dyn_Money_Top_A_02}
    componentAppliedLinearAccClamp=<3000 3000 3000>
    componentLinearAccScale=<0.4 0.4 0.4>
    componentLinearVelScale=<0.4 0.4 0.4>
    gravityOverride {z=-980}
    physicsBodyDefinitions [{
      boundBone {boneName=dyn_Money_Top_A_02}
      boxExtents=<20 20 20>
      constraintSetup {angularLimitsMax=<0 5 7>, angularLimitsMin=<0 -5 -25>, angularTarget=<0 0 0>}
      localJointOffset=<8 0 0>
     }]
    simSpaceSettings {simSpaceAngularAlpha=1}
   }
   n5 = Anim Dynamics - Bone: dyn_Money_Top_B_01!(ComponentPose=@n4, Alpha=1) -> (Pose) {
    angularSpringConstant=15000
    bAngularSpring=true
    bUseGravityOverride=true
    simulationSpace=RootRelative
    boundBone {boneName=dyn_Money_Top_B_01}
    componentAppliedLinearAccClamp=<3000 3000 3000>
    componentLinearAccScale=<0.5 0.5 0.5>
    componentLinearVelScale=<0.5 0.5 0.5>
    gravityOverride {z=-980}
    physicsBodyDefinitions [{
      boundBone {boneName=dyn_Money_Top_B_01}
      boxExtents=<40 40 40>
      constraintSetup {angularLimitsMax=<0 0 0>, angularLimitsMin=<0 0 -15>, angularTarget=<0 0 0>}
      localJointOffset=<2 0 0>
     }]
    simSpaceSettings {simSpaceAngularAlpha=1}
   }
   n6 = Anim Dynamics - Bone: dyn_Money_Top_B_02!(ComponentPose=@n5, Alpha=1) -> (Pose) {
    angularSpringConstant=15000
    bAngularSpring=true
    bUseGravityOverride=true
    simulationSpace=RootRelative
    boundBone {boneName=dyn_Money_Top_B_02}
    componentAppliedLinearAccClamp=<3000 3000 3000>
    componentLinearAccScale=<0.5 0.5 0.5>
    componentLinearVelScale=<0.5 0.5 0.5>
    gravityOverride {z=-980}
    physicsBodyDefinitions [{
      boundBone {boneName=dyn_Money_Top_B_02}
      boxExtents=<20 20 20>
      constraintSetup {angularLimitsMax=<0 0 20>, angularLimitsMin=<0 0 -10>, angularTarget=<0 0 0>}
      localJointOffset=<8 0 0>
     }]
   }
   n7 = Anim Dynamics - Bone: dyn_Money_Top_C_01!(ComponentPose=@n6, Alpha=1) -> (Pose) {
    angularSpringConstant=15000
    bAngularSpring=true
    bUseGravityOverride=true
    simulationSpace=RootRelative
    boundBone {boneName=dyn_Money_Top_C_01}
    componentAppliedLinearAccClamp=<3000 3000 3000>
    componentLinearAccScale=<0.5 0.5 0.5>
    componentLinearVelScale=<0.5 0.5 0.5>
    gravityOverride {z=-980}
    physicsBodyDefinitions [{
      boundBone {boneName=dyn_Money_Top_C_01}
      boxExtents=<40 40 40>
      constraintSetup {angularLimitsMax=<0 5 10>, angularLimitsMin=<0 0 -15>, angularTarget=<0 0 0>}
      localJointOffset=<2 0 0>
     }]
    simSpaceSettings {simSpaceAngularAlpha=1}
   }
   n8 = Anim Dynamics - Bone: dyn_Money_Top_C_02!(ComponentPose=@n7, Alpha=1) -> (Pose) {
    angularSpringConstant=15000
    bAngularSpring=true
    bUseGravityOverride=true
    simulationSpace=RootRelative
    boundBone {boneName=dyn_Money_Top_C_02}
    componentAppliedLinearAccClamp=<3000 3000 3000>
    componentLinearAccScale=<0.5 0.5 0.5>
    componentLinearVelScale=<0.5 0.5 0.5>
    gravityOverride {z=-980}
    physicsBodyDefinitions [{
      boundBone {boneName=dyn_Money_Top_C_02}
      boxExtents=<15 15 15>
      constraintSetup {angularLimitsMax=<0 0 25>, angularLimitsMin=<0 0 0>, angularTarget=<0 0 0>}
      localJointOffset=<8 0 0>
     }]
   }
   n9 = Anim Dynamics - Bone: dyn_Money_Bot_D_01!(ComponentPose=@n8, Alpha=1) -> (Pose) {
    angularSpringConstant=15000
    bAngularSpring=true
    bUseGravityOverride=true
    simulationSpace=BoneRelative
    boundBone {boneName=dyn_Money_Bot_D_01}
    componentAppliedLinearAccClamp=<3000 3000 3000>
    componentLinearAccScale=<0.5 0.5 0.5>
    componentLinearVelScale=<0.5 0.5 0.5>
    gravityOverride {z=-980}
    physicsBodyDefinitions [{
      boundBone {boneName=dyn_Money_Bot_D_01}
      boxExtents=<20 20 20>
      constraintSetup {angularLimitsMax=<0 10 25>, angularLimitsMin=<0 -10 -25>, angularTarget=<0 0 0>}
      localJointOffset=<4 0 0>
     }]
    relativeSpaceBone {boneName=handle}
    simSpaceSettings {simSpaceAngularAlpha=1}
   }
   n10 = Anim Dynamics - Bone: dyn_Money_Bot_D_02!(ComponentPose=@n9, Alpha=1) -> (Pose) {
    angularSpringConstant=15000
    bAngularSpring=true
    bEnableWind=true
    bUseGravityOverride=true
    simulationSpace=RootRelative
    boundBone {boneName=dyn_Money_Bot_D_02}
    componentLinearAccScale=<0.2 0.2 0.2>
    gravityOverride {z=-980}
    physicsBodyDefinitions [{
      boundBone {boneName=dyn_Money_Bot_D_02}
      boxExtents=<20 20 20>
      constraintSetup {angularLimitsMax=<0 0 50>, angularLimitsMin=<0 0 -50>, angularTarget=<0 0 0>}
      localJointOffset=<6 0 0>
     }]
    simSpaceSettings {simSpaceAngularAlpha=1}
   }
   n11 = Anim Dynamics - Bone: dyn_Money_Bot_E_01!(ComponentPose=@n10, Alpha=1) -> (Pose) {
    angularSpringConstant=25000
    bAngularSpring=true
    bUseGravityOverride=true
    simulationSpace=RootRelative
    boundBone {boneName=dyn_Money_Bot_E_01}
    componentAppliedLinearAccClamp=<3000 3000 3000>
    componentLinearAccScale=<0.5 0.5 0.5>
    componentLinearVelScale=<0.5 0.5 0.5>
    gravityOverride {z=-1500}
    physicsBodyDefinitions [{
      boundBone {boneName=dyn_Money_Bot_E_01}
      boxExtents=<20 20 20>
      constraintSetup {angularLimitsMax=<0 10 15>, angularLimitsMin=<0 -10 -15>, angularTarget=<0 0 0>}
      localJointOffset=<8 0 0>
     }]
    simSpaceSettings {simSpaceAngularAlpha=1}
   }
   n12 = Anim Dynamics - Bone: dyn_Money_Bot_E_02!(ComponentPose=@n11, Alpha=1) -> (Pose) {
    angularSpringConstant=15000
    bAngularSpring=true
    bUseGravityOverride=true
    simulationSpace=RootRelative
    boundBone {boneName=dyn_Money_Bot_E_02}
    gravityOverride {z=-980}
    physicsBodyDefinitions [{
      boundBone {boneName=dyn_Money_Bot_E_02}
      boxExtents=<15 15 15>
      constraintSetup {angularLimitsMax=<0 10 10>, angularLimitsMin=<0 -10 -10>, angularTarget=<0 0 0>}
      localJointOffset=<2 0 0>
     }]
    simSpaceSettings {simSpaceAngularAlpha=1}
   }
   n13 = Anim Dynamics - Bone: dyn_handle_01!(ComponentPose=@n12, Alpha=1, GravityOverride:Vector="0,0,-4000") -> (Pose) {
    angularDampingOverride=0.9
    angularSpringConstant=15000
    bAngularSpring=true
    bOverrideAngularDamping=true
    bUseGravityOverride=true
    simulationSpace=RootRelative
    boundBone {boneName=dyn_handle_01}
    componentAppliedLinearAccClamp=<3000 3000 3000>
    componentLinearAccScale=<0.5 0.5 0.5>
    componentLinearVelScale=<0.5 0.5 0.5>
    gravityOverride {z=-4000}
    physicsBodyDefinitions [{
      boundBone {boneName=dyn_handle_01}
      boxExtents=<40 40 40>
      constraintSetup {angularLimitsMax=<0 0 120>, angularLimitsMin=<0 0 0>, angularTarget=<0 0 0>}
      localJointOffset=<5 0 0>
     }]
    simSpaceSettings {maxAngularAcceleration=32, maxAngularVelocity=8, simSpaceAngularAlpha=1}
   }
   Output Pose(Result=Component To Local(ComponentPose=@n13))
  ]}
 EventGraph {nodes [
   Event BlueprintUpdateAnimation()
   > n1 = Branch(Condition=Get bIsInFrontEnd())
   > Set AlphaBills(AlphaBills:double=0.6) -> (Output_Get:double)
   > Set Override World Gravity(Override World Gravity:Vector="0,0,-4100") -> (Output_Get:Vector)
   > n2 = Branch(Condition=Get bIsPlayingEmote())
   Set HandleAlpha(execute=@n2.else, HandleAlpha:double=0.88) -> (Output_Get:double)
   Set AlphaBills(execute=@n1.else, AlphaBills=1)
   > Set Override World Gravity(Override World Gravity="0,0,-1500")
   Event OnBeginPlayingEmote() {is=title K2Node_Event}
   > Set HandleAlpha(HandleAlpha=0)
   Event OnEndPlayingEmote() {is=title K2Node_Event}
   > Set HandleAlpha(HandleAlpha=1)
  ]}
]
variables [
 AlphaBills {type=number}
 HandleAlpha {type=number, default=1}
 Override World Gravity {type=Vector, default=<0 0 -1500>}
 rigidBodyOff {type=boolean}
]
